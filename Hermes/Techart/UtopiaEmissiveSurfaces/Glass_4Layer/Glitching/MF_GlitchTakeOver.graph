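# Material function for glitch takeover effects on emissive glass, outputting an intermittent glitch mask, square masks, randomized time and glitched UVs.
# A sine offset by instance random, custom primitive data and actor position is thresholded, multiplied by squared flicker and evaluated through a VertexInterpolator.
# Hashes of rotated, gridded UVs and time form square masks and select blocks that snap to varying pixelation resolutions.
# A trailing panned-texture color glitch branch reaches no output.
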MaterialFunction /Hermes/Techart/UtopiaEmissiveSurfaces/Glass_4Layer/Glitching/MF_GlitchTakeOver
graph {nodes [
  Constant() {r=-0.8415}
  n1 = Constant() {r=1}
  n2 = FunctionInput(Preview=Constant() {r=0.075}) {bUsePreviewValueAsDefault=true, inputName=GlitchDuration, inputType=FunctionInput_Scalar}
  n3 = PerInstanceRandom()
  n4 = ScalarParameter() {bUseCustomPrimitiveData=true, parameterName=Offset Time, primitiveDataIndex=2}
  n5 = MaterialFunctionCall(Float3=Divide(A=ActorPositionWS()) {constB=2000}) {materialFunction=&BreakOutFloat3Components}
  n6 = Add(A=@n5.R, B=OneMinus(None=@n5.G))
  n7 = Add(A=Add(A=LinearInterpolate(Alpha=@n3) {constB=12335.3467}, B=Add(A=@n4, B=@n6)), B=Time())
  n8 = FunctionInput(Preview=Constant() {r=0.1}) {bUsePreviewValueAsDefault=true, inputName=GlitchFrequency, inputType=FunctionInput_Scalar}
  n9 = Multiply(A=Sine(None=Multiply(A=@n7, B=@n8) {constB=0.2})) {constB=0.5}
  n10 = Constant()
  n11 = Saturate(None=Add(
    A=MaterialFunctionCall() {materialFunction=/Game/MF__FloatFlickerRandom_Generic}.RandomFlickerFloat
    B=FunctionInput(Preview=@n10) {bUsePreviewValueAsDefault=true, inputName=Glitch Flickering, inputType=FunctionInput_Scalar}
   ))
  n12 = Multiply(A=SmoothStep(Min=Subtract(A=@n1, B=@n2), Max=@n1, Value=Add(A=@n9) {constB=0.5}) {constMax=0.9, constMin=0.8}, B=Multiply(A=@n11, B=@n11))
  FunctionOutput(None=NamedRerouteDeclaration(None=VertexInterpolator(VS=@n12).PS) {name=GlitchFrequencyMask, nodeColor=<rgba 0.9377 1 0.8953 1>}) {outputName=GlitchFrequencyMask}
  n13 = FunctionInput(Preview=NamedRerouteUsage()) {bUsePreviewValueAsDefault=true, inputName=Time, inputType=FunctionInput_Scalar}
  n14 = FunctionInput(Preview=TextureCoordinate()) {bUsePreviewValueAsDefault=true, inputName=UV, inputType=FunctionInput_Vector2}
  n15 = MaterialFunctionCall(UVs=@n14, "Rotation Angle (0-1)"=ScalarParameter() {group=HolographicTexture_Glitch, parameterName=GlitchUVRotation}) {materialFunction=&CustomRotator}
  n16 = Multiply(
   A=@n15.Rotated Values
   B=FunctionInput(Preview=ScalarParameter() {defaultValue=2.5, group=HolographicTexture_Glitch, parameterName=GlitchMaskResolution}) {bUsePreviewValueAsDefault=true, inputName=GlitchMaskResolution, inputType=FunctionInput_Scalar}
  )
  n17 = MaterialFunctionCall!(Seed=@n13, Time=@n13, Vec3, Vec2=Floor(None=Add(A=@n16))) {materialFunction=/Hermes/Techart/UtopiaEmissiveSurfaces/MF_CustomHash}
  n18 = ComponentMask(None=ConstantBiasScale(None=@n17.Vec2)) {r=true}
  n19 = Add(A=@n13) {constB=0.5}
  n20 = MaterialFunctionCall(Time=@n19) {materialFunction=&MF_CustomHash}
  n21 = Multiply(A=@n18, B=Add(A=@n18, B=ComponentMask(None=@n20.Vec2) {g=true}))
  FunctionOutput(None=NamedRerouteDeclaration(None=FunctionOutput(None=@n21) {outputName=Square Mask}) {name=Sqaure Color glitch, nodeColor=<rgba 0.2422 1 0 1>}) {outputName=Sqaure Color glitch}
  n22 = ScalarParameter() {bUseCustomPrimitiveData=true, parameterName=Offset Time, primitiveDataIndex=2}
  n23 = NamedRerouteUsage()
  n24 = ComponentMask(None=TransformPosition(None=Constant3Vector()) {transformType=TRANSFORMPOSSOURCE_World}) {g=true, r=true}
  n25 = Ceil(None=AppendVector(A=Add(A=@n23, B=@n24), B=Constant()))
  n26 = Constant() {desc="LeapOfFaith ", r=1000}
  n27 = MaterialFunctionCall(Float3=Multiply(A=Divide(A=@n25, B=@n26)) {constB=11.17}) {materialFunction=&BreakOutFloat3Components}
  n28 = Add(A=@n27.R, B=@n27.R)
  n29 = Add(A=Multiply(A=Frac(None=@n28)) {constB=120}, B=Time())
  FunctionOutput(None=NamedRerouteDeclaration(None=Add(A=@n22, B=@n29)) {name=Random time, nodeColor=<rgba 0 1 0.7422 1>}) {outputName=Random time}
  n30 = LinearInterpolate(
   A=FunctionInput(Preview=ScalarParameter() {defaultValue=12, group=HolographicTexture_Glitch, parameterName=PixellateGlitch_MaxResolution}) {bUsePreviewValueAsDefault=true, inputName=PixellateGlitch_MaxResolution, inputType=FunctionInput_Scalar}
   B=FunctionInput(Preview=ScalarParameter() {defaultValue=4, group=HolographicTexture_Glitch, parameterName=PixellateGlitch_MinResolution}) {bUsePreviewValueAsDefault=true, inputName=PixellateGlitch_MinResolution, inputType=FunctionInput_Scalar}
   Alpha=Round(None=Clamp(None=@n17.Float))
  )
  n31 = Divide(A=Floor(None=Multiply(A=@n15.Rotated Values, B=@n30)), B=@n30)
  n32 = Round(None=@n18)
  n33 = ConstantBiasScale(None=@n20.Float)
  FunctionOutput(None=LinearInterpolate(A=LinearInterpolate(A=@n14, B=@n31, Alpha=@n32), B=@n14, Alpha=Round(None=@n33))) {outputName=Glitched UV}
  MaterialFunctionCall!(X range Y range, Gradient Height, ShowGrid?, Thickness, Color, "f(x)"=OneMinus(None=@n12)) {materialFunction=/Engine/Functions/Engine_MaterialFunctions01/Debug/PlotFunctionOnGraph}
  Multiply(A=@n19, B=Constant() {r=0.67})
  NamedRerouteDeclaration(None=FunctionInput(Preview=TextureCoordinate()) {inputName=Triplanar UV, inputType=FunctionInput_Vector2, sortPriority=4}) {name=Trip UV, nodeColor=<rgba 0 1 0.8359 1>}
  NamedRerouteUsage()
  n34 = NamedRerouteUsage()
  n35 = FunctionInput(Preview=Constant() {r=2}) {bUsePreviewValueAsDefault=true, inputName=Glitch Size, inputType=FunctionInput_Scalar, sortPriority=4}
  n36 = Constant() {r=2}
  n37 = Divide(A=Divide(
    A=Reroute()
    B=FunctionInput(Preview=@n36) {bUsePreviewValueAsDefault=true, inputName=Color Glitch Speed, inputType=FunctionInput_Scalar, sortPriority=4}
   )) {constB=2.37}
  n38 = TextureObject() {samplerType=SAMPLERTYPE_LinearColor, texture=/Game/Effects/Fort_Effects/Effects/Characters/Generic/ResIn/DiffusedStylizedGridAnimOffsets}
  n39 = TextureCoordinate()
  n40 = TextureSample!(
   UVs=Panner(Coordinate=Divide(A=@n34, B=@n35) {constB=10}, Time=Floor(None=@n37)) {speedX=0.3, speedY=0.1}
   Tex=FunctionInput(Preview=@n38) {bUsePreviewValueAsDefault=true, inputName=Texture Glitch, inputType=FunctionInput_Texture2D, sortPriority=5}
   "DDX(UVs)"=FunctionInput(Preview=DDX(Value=@n39)) {inputName=DDX, inputType=FunctionInput_Vector2, sortPriority=2}
   "DDY(UVs)"=FunctionInput(Preview=DDY(Value=@n39)) {inputName=DDY, inputType=FunctionInput_Vector2, sortPriority=3}
   Apply View MipBias
  ) {
   mipValueMode=TMVM_Derivative
   samplerType=SAMPLERTYPE_LinearColor
   texture=&DiffusedStylizedGridAnimOffsets
  }
  n41 = Multiply(A=NamedRerouteUsage(), B=OneMinus(None=@n40.R))
  Saturate(None=Multiply(A=OneMinus(None=@n41)) {constB=2})
  ScalarParameter() {defaultValue=0.1, group=HolographicTexture_Glitch, parameterName=GlitchFrequency, sortPriority=10}
  ScalarParameter() {defaultValue=0.075, group=HolographicTexture_Glitch, parameterName=GlitchDuration, sortPriority=10}
 ]}
