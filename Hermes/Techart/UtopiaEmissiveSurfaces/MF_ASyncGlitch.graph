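# Material function that produces a time-driven blocky pixelation glitch on UVs, with parameters grouped for holographic textures.
# Rotated UVs, floored into a grid, feed MF_CustomHash to build the Square Mask output and pick maximum or minimum pixelation resolution per cell.
# The rounded mask blends original UVs with snapped rotated UVs.
# A second hash on offset time, independent of UV, switches the whole Glitched UV output back to the original intermittently.
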MaterialFunction /Hermes/Techart/UtopiaEmissiveSurfaces/MF_ASyncGlitch
graph {nodes [
  n1 = FunctionInput(Preview=Divide(A=Time()) {constB=5}) {bUsePreviewValueAsDefault=true, inputName=Time, inputType=FunctionInput_Scalar}
  n2 = Add(A=@n1) {constB=0.5}
  n3 = MaterialFunctionCall!(Seed, Time=@n2, Vec3, Vec2) {materialFunction=&/MF_CustomHash}
  ComponentMask(None=@n3.Vec2) {g=true, r=true}
  n4 = FunctionInput(Preview=TextureCoordinate()) {bUsePreviewValueAsDefault=true, inputName=UV, inputType=FunctionInput_Vector2}
  n5 = MaterialFunctionCall(UVs=@n4, "Rotation Angle (0-1)"=ScalarParameter() {group=HolographicTexture_Glitch, parameterName=GlitchUVRotation}) {materialFunction=&CustomRotator}
  n6 = Multiply(
   A=@n5.Rotated Values
   B=FunctionInput(Preview=ScalarParameter() {defaultValue=2.5, group=HolographicTexture_Glitch, parameterName=GlitchMaskResolution}) {bUsePreviewValueAsDefault=true, inputName=GlitchMaskResolution, inputType=FunctionInput_Scalar}
  )
  n7 = MaterialFunctionCall(Seed=@n1, Time=@n1, Vec2=Floor(None=Add(A=@n6))) {materialFunction=&MF_CustomHash}
  n8 = ComponentMask(None=ConstantBiasScale(None=@n7.Vec2)) {r=true}
  FunctionOutput(None=@n8) {outputName=Square Mask}
  n9 = LinearInterpolate(
   A=FunctionInput(Preview=ScalarParameter() {defaultValue=12, group=HolographicTexture_Glitch, parameterName=PixellateGlitch_MaxResolution}) {bUsePreviewValueAsDefault=true, inputName=PixellateGlitch_MaxResolution, inputType=FunctionInput_Scalar}
   B=FunctionInput(Preview=ScalarParameter() {defaultValue=4, group=HolographicTexture_Glitch, parameterName=PixellateGlitch_MinResolution}) {bUsePreviewValueAsDefault=true, inputName=PixellateGlitch_MinResolution, inputType=FunctionInput_Scalar}
   Alpha=Round(None=Clamp(None=@n7.Float))
  )
  n10 = Divide(A=Floor(None=Multiply(A=@n5.Rotated Values, B=@n9)), B=@n9)
  n11 = Round(None=@n8)
  n12 = ConstantBiasScale(None=@n3.Float)
  FunctionOutput(None=LinearInterpolate(A=LinearInterpolate(A=@n4, B=@n10, Alpha=@n11), B=@n4, Alpha=Round(None=@n12))) {outputName=Glitched UV}
  Multiply(A=@n2, B=Constant() {r=0.67})
  Time()
 ]}
