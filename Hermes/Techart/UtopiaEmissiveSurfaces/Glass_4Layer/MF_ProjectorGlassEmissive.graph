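# Material function that adds a projector-style emissive effect to glass surfaces, taking existing material attributes and an SFX input.
# It builds layered imagery from glitch-distorted UVs, square masks, time-driven motion, and a world position offset onto a plane through view-dependent dot products.
# Pixel-depth terms, a selectable glass mask channel, the Fortnite time-of-day parameter and inverse eye adaptation feed the shading.
# Static switches choose variants, and a masked material-attribute blend combines results.
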MaterialFunction /Hermes/Techart/UtopiaEmissiveSurfaces/Glass_4Layer/MF_ProjectorGlassEmissive
graph {expressions {
  MaterialExpressionAbs_1 {class=MaterialExpressionAbs, inputs [None {from=MaterialExpressionAdd_53, output=""}], properties {}}
  MaterialExpressionAbs_5 {class=MaterialExpressionAbs, inputs [None {from=MaterialExpressionAdd_54, output=""}], properties {}}
  MaterialExpressionAbs_6 {class=MaterialExpressionAbs, inputs [None {from=MaterialExpressionAdd_55, output=""}], properties {}}
  MaterialExpressionAbs_7 {class=MaterialExpressionAbs, inputs [None {from=MaterialExpressionAdd_56, output=""}], properties {}}
  MaterialExpressionAbs_8 {class=MaterialExpressionAbs, inputs [None {from=MaterialExpressionAdd_44, output=""}], properties {}}
  MaterialExpressionAdd_0 {
   class=MaterialExpressionAdd
   inputs [
    A {from=MaterialExpressionMultiply_125}
    B {from=MaterialExpressionMultiply_0}
   ]
  }
  MaterialExpressionAdd_1 {
   class=MaterialExpressionAdd
   inputs [
    A {from=MaterialExpressionLinearInterpolate_6}
    B {from=MaterialExpressionLinearInterpolate_8}
   ]
  }
  MaterialExpressionAdd_11 {
   class=MaterialExpressionAdd
   inputs [
    A {from=MaterialExpressionNamedRerouteUsage_42}
    B {from=MaterialExpressionNamedRerouteUsage_54}
   ]
  }
  MaterialExpressionAdd_12 {
   class=MaterialExpressionAdd
   inputs [
    A {from=MaterialExpressionMultiply_51}
    B {from=""}
   ]
   properties {constB=0.5}
  }
  MaterialExpressionAdd_13 {
   class=MaterialExpressionAdd
   inputs [
    A {from=MaterialExpressionNamedRerouteUsage_42}
    B {from=MaterialExpressionNamedRerouteUsage_8}
   ]
  }
  MaterialExpressionAdd_14 {
   class=MaterialExpressionAdd
   inputs [
    A {from=MaterialExpressionMaterialFunctionCall_14, output=Square Mask}
    B {from=""}
   ]
   properties {constB=0.1}
  }
  MaterialExpressionAdd_15 {
   class=MaterialExpressionAdd
   inputs [
    A {from=MaterialExpressionMultiply_20}
    B {from=""}
   ]
   properties {constB=0.6}
  }
  MaterialExpressionAdd_17 {
   class=MaterialExpressionAdd
   inputs [
    A {from=MaterialExpressionMultiply_21}
    B {from=""}
   ]
   properties {constB=0.3}
  }
  MaterialExpressionAdd_18 {
   class=MaterialExpressionAdd
   inputs [
    A {from=MaterialExpressionNamedRerouteUsage_72}
    B {from=MaterialExpressionMaterialFunctionCall_14, output=Glitched UV}
   ]
  }
  MaterialExpressionAdd_19 {
   class=MaterialExpressionAdd
   inputs [
    A {from=MaterialExpressionWorldPosition_1, output=XYZ}
    B {from=MaterialExpressionDivide_3}
   ]
  }
  MaterialExpressionAdd_2 {
   class=MaterialExpressionAdd
   inputs [
    A {from=MaterialExpressionSaturate_33}
    B {from=""}
   ]
   properties {constB=-0.14}
  }
  MaterialExpressionAdd_23 {
   class=MaterialExpressionAdd
   inputs [
    A {from=MaterialExpressionMultiply_117}
    B {from=MaterialExpressionNamedRerouteUsage_89}
   ]
  }
  MaterialExpressionAdd_3 {
   class=MaterialExpressionAdd
   inputs [
    A {from=MaterialExpressionComponentMask_5}
    B {from=MaterialExpressionDivide_1}
   ]
  }
  MaterialExpressionAdd_30 {
   class=MaterialExpressionAdd
   inputs [
    A {from=MaterialExpressionFrac_0}
    B {from=""}
   ]
   properties {constB=-0.5}
  }
  MaterialExpressionAdd_37 {
   class=MaterialExpressionAdd
   inputs [
    A {from=MaterialExpressionMaterialFunctionCall_7, output=R}
    B {from=MaterialExpressionMaterialFunctionCall_7, output=R}
   ]
  }
  MaterialExpressionAdd_39 {
   class=MaterialExpressionAdd
   inputs [
    A {from=MaterialExpressionMultiply_106}
    B {from=MaterialExpressionMultiply_110}
   ]
   properties {constB=-2}
  }
  MaterialExpressionAdd_4 {
   class=MaterialExpressionAdd
   inputs [
    A {from=MaterialExpressionMaterialFunctionCall_44, output=Result}
    B {from=MaterialExpressionNamedRerouteUsage_9}
   ]
  }
  MaterialExpressionAdd_41 {
   class=MaterialExpressionAdd
   inputs [
    A {from=MaterialExpressionMultiply_112}
    B {from=MaterialExpressionReroute_10}
   ]
  }
  MaterialExpressionAdd_42 {
   class=MaterialExpressionAdd
   inputs [
    A {from=MaterialExpressionMultiply_158}
    B {from=MaterialExpressionNamedRerouteUsage_162}
   ]
  }
  MaterialExpressionAdd_43 {
   class=MaterialExpressionAdd
   inputs [
    A {from=MaterialExpressionNamedRerouteUsage_163}
    B {from=MaterialExpressionMultiply_159}
   ]
  }
  MaterialExpressionAdd_44 {
   class=MaterialExpressionAdd
   inputs [
    A {from=MaterialExpressionRotator_0}
    B {from=""}
   ]
   properties {constB=-0.5}
  }
  MaterialExpressionAdd_45 {
   class=MaterialExpressionAdd
   inputs [
    A {from=MaterialExpressionNamedRerouteUsage_94}
    B {from=MaterialExpressionMultiply_122}
   ]
  }
  MaterialExpressionAdd_46 {
   class=MaterialExpressionAdd
   inputs [
    A {from=MaterialExpressionNamedRerouteUsage_107}
    B {from=MaterialExpressionMultiply_123}
   ]
  }
  MaterialExpressionAdd_47 {
   class=MaterialExpressionAdd
   inputs [
    A {from=MaterialExpressionMultiply_156}
    B {from=MaterialExpressionNamedRerouteUsage_156}
   ]
  }
  MaterialExpressionAdd_48 {
   class=MaterialExpressionAdd
   inputs [
    A {from=MaterialExpressionNamedRerouteUsage_157}
    B {from=MaterialExpressionMultiply_157}
   ]
  }
  MaterialExpressionAdd_49 {
   class=MaterialExpressionAdd
   inputs [
    A {from=MaterialExpressionSaturate_47}
    B {from=MaterialExpressionSaturate_47}
   ]
  }
  MaterialExpressionAdd_5 {
   class=MaterialExpressionAdd
   inputs [
    A {from=MaterialExpressionLinearInterpolate_4}
    B {from=MaterialExpressionMultiply_7}
   ]
  }
  MaterialExpressionAdd_50 {
   class=MaterialExpressionAdd
   inputs [
    A {from=MaterialExpressionMultiply_4}
    B {from=MaterialExpressionPower_5}
   ]
  }
  MaterialExpressionAdd_51 {
   class=MaterialExpressionAdd
   inputs [
    A {from=MaterialExpressionScalarParameter_74}
    B {from=""}
   ]
   properties {constB=2}
  }
  MaterialExpressionAdd_52 {
   class=MaterialExpressionAdd
   inputs [
    A {from=MaterialExpressionSaturate_44}
    B {from=MaterialExpressionMultiply_145}
   ]
  }
  MaterialExpressionAdd_53 {
   class=MaterialExpressionAdd
   inputs [
    A {from=MaterialExpressionComponentMask_8}
    B {from=""}
   ]
   properties {constB=-0.5}
  }
  MaterialExpressionAdd_54 {
   class=MaterialExpressionAdd
   inputs [
    A {from=MaterialExpressionAdd_53}
    B {from=""}
   ]
   properties {constB=-1}
  }
  MaterialExpressionAdd_55 {
   class=MaterialExpressionAdd
   inputs [
    A {from=MaterialExpressionAdd_54}
    B {from=""}
   ]
   properties {constB=-1}
  }
  MaterialExpressionAdd_56 {
   class=MaterialExpressionAdd
   inputs [
    A {from=MaterialExpressionAdd_55}
    B {from=""}
   ]
   properties {constB=-1}
  }
  MaterialExpressionAdd_57 {
   class=MaterialExpressionAdd
   inputs [
    A {from=MaterialExpressionLinearInterpolate_52}
    B {from=MaterialExpressionNamedRerouteUsage_126}
   ]
  }
  MaterialExpressionAdd_59 {
   class=MaterialExpressionAdd
   inputs [
    A {from=MaterialExpressionOneMinus_23}
    B {from=""}
   ]
   properties {constB=0.9}
  }
  MaterialExpressionAdd_6 {
   class=MaterialExpressionAdd
   inputs [
    A {from=MaterialExpressionComponentMask_3}
    B {from=""}
   ]
  }
  MaterialExpressionAdd_63 {
   class=MaterialExpressionAdd
   inputs [
    A {from=MaterialExpressionScalarParameter_95}
    B {from=""}
   ]
  }
  MaterialExpressionAdd_64 {
   class=MaterialExpressionAdd
   inputs [
    A {from=MaterialExpressionScalarParameter_96}
    B {from=""}
   ]
  }
  MaterialExpressionAdd_65 {
   class=MaterialExpressionAdd
   inputs [
    A {from=MaterialExpressionScalarParameter_100}
    B {from=""}
   ]
  }
  MaterialExpressionAdd_66 {
   class=MaterialExpressionAdd
   inputs [
    A {from=MaterialExpressionScalarParameter_98}
    B {from=""}
   ]
  }
  MaterialExpressionAdd_7 {
   class=MaterialExpressionAdd
   inputs [
    A {from=MaterialExpressionTextureSampleParameter2D_0, output=G}
    B {from=MaterialExpressionTextureSampleParameter2D_0, output=G}
   ]
  }
  MaterialExpressionAdd_71 {
   class=MaterialExpressionAdd
   inputs [
    A {from=MaterialExpressionAdd_45}
    B {from=MaterialExpressionMultiply_198}
   ]
  }
  MaterialExpressionAdd_76 {
   class=MaterialExpressionAdd
   inputs [
    A {from=MaterialExpressionMultiply_204}
    B {from=MaterialExpressionSubtract_3}
   ]
  }
  MaterialExpressionAdd_8 {
   class=MaterialExpressionAdd
   inputs [
    A {from=""}
    B {from=""}
   ]
  }
  MaterialExpressionAdd_9 {
   class=MaterialExpressionAdd
   inputs [
    A {from=""}
    B {from=""}
   ]
  }
  MaterialExpressionAppendVector_2 {
   class=MaterialExpressionAppendVector
   inputs [
    A {from=MaterialExpressionComponentMask_7, output=""}
    B {from=MaterialExpressionAdd_3, output=""}
   ]
  }
  MaterialExpressionAppendVector_4 {
   class=MaterialExpressionAppendVector
   inputs [
    A {from=MaterialExpressionComponentMask_1, output=""}
    B {from=MaterialExpressionComponentMask_3, output=""}
   ]
  }
  MaterialExpressionBlendMaterialAttributes_0 {
   class=MaterialExpressionBlendMaterialAttributes
   inputs [
    A {from=MaterialExpressionNamedRerouteDeclaration_59, output=""}
    B {from=MaterialExpressionStaticSwitchParameter_30, output=""}
    Alpha {from=MaterialExpressionMultiply_162, output=""}
   ]
   properties {}
  }
  MaterialExpressionChannelMaskParameter_1 {
   class=MaterialExpressionChannelMaskParameter
   inputs [None {from=MaterialExpressionMaterialFunctionCall_33, output=Result}]
   properties {group=Projector, parameterName=Glass Mask Channel, sortPriority=0}
  }
  MaterialExpressionCollectionParameter_3 {
   class=MaterialExpressionCollectionParameter
   inputs []
   properties {
    collection=&FortniteMaterialParameters
    parameterName=FortniteActiveTimeOfDay
   }
  }
  MaterialExpressionComponentMask_0 {class=MaterialExpressionComponentMask, inputs [None {from=""}], properties {g=true, r=true}}
  MaterialExpressionComponentMask_1 {class=MaterialExpressionComponentMask, inputs [None {from=MaterialExpressionVertexInterpolator_2, output=PS}], properties {r=true}}
  MaterialExpressionComponentMask_2 {class=MaterialExpressionComponentMask, inputs [None {from=MaterialExpressionVertexInterpolator_1, output=PS}], properties {g=true, r=true}}
  MaterialExpressionComponentMask_3 {class=MaterialExpressionComponentMask, inputs [None {from=MaterialExpressionVertexInterpolator_2, output=PS}], properties {g=true}}
  MaterialExpressionComponentMask_5 {class=MaterialExpressionComponentMask, inputs [None {from=MaterialExpressionMultiply_37}], properties {g=true}}
  MaterialExpressionComponentMask_6 {class=MaterialExpressionComponentMask, inputs [None {from=MaterialExpressionNamedRerouteUsage_105}], properties {g=true, r=true}}
  MaterialExpressionComponentMask_7 {class=MaterialExpressionComponentMask, inputs [None {from=MaterialExpressionMultiply_37}], properties {r=true}}
  MaterialExpressionComponentMask_8 {class=MaterialExpressionComponentMask, inputs [None {from=MaterialExpressionTextureCoordinate_5}], properties {r=true}}
  MaterialExpressionConstant3Vector_0 {class=MaterialExpressionConstant3Vector, properties {constant=<rgba 0 0 1 0>}}
  MaterialExpressionConstant3Vector_1 {class=MaterialExpressionConstant3Vector, properties {constant=<rgba 0.0382 0.0802 0.1022 1>}}
  MaterialExpressionConstant_0 {class=MaterialExpressionConstant, properties {r=0.1}}
  MaterialExpressionConstant_1 {class=MaterialExpressionConstant, properties {r=1}}
  MaterialExpressionConstant_15 {class=MaterialExpressionConstant, properties {r=1}}
  MaterialExpressionConstant_16 {class=MaterialExpressionConstant, properties {r=0.33}}
  MaterialExpressionConstant_17 {class=MaterialExpressionConstant, properties {r=0.38}}
  MaterialExpressionConstant_2 {class=MaterialExpressionConstant, properties {r=5}}
  MaterialExpressionConstant_27 {class=MaterialExpressionConstant, properties {r=0.85}}
  MaterialExpressionConstant_28 {class=MaterialExpressionConstant, properties {r=6}}
  MaterialExpressionConstant_29 {class=MaterialExpressionConstant, properties {r=-1.9818}}
  MaterialExpressionConstant_30 {class=MaterialExpressionConstant, properties {r=7.1969}}
  MaterialExpressionConstant_32 {class=MaterialExpressionConstant}
  MaterialExpressionConstant_33 {class=MaterialExpressionConstant, properties {r=0.1}}
  MaterialExpressionConstant_34 {class=MaterialExpressionConstant, properties {r=0.016}}
  MaterialExpressionConstant_35 {class=MaterialExpressionConstant, properties {r=0.5}}
  MaterialExpressionConstant_36 {class=MaterialExpressionConstant}
  MaterialExpressionConstant_37 {class=MaterialExpressionConstant, properties {r=5}}
  MaterialExpressionConstant_38 {class=MaterialExpressionConstant}
  MaterialExpressionConstant_39 {class=MaterialExpressionConstant, properties {r=2}}
  MaterialExpressionConstant_4 {class=MaterialExpressionConstant}
  MaterialExpressionConstant_40 {class=MaterialExpressionConstant}
  MaterialExpressionConstant_41 {class=MaterialExpressionConstant}
  MaterialExpressionConstant_42 {class=MaterialExpressionConstant}
  MaterialExpressionConstant_43 {class=MaterialExpressionConstant}
  MaterialExpressionConstant_44 {class=MaterialExpressionConstant}
  MaterialExpressionConstant_45 {class=MaterialExpressionConstant, properties {r=3}}
  MaterialExpressionConstant_46 {class=MaterialExpressionConstant}
  MaterialExpressionConstant_51 {class=MaterialExpressionConstant, properties {r=1}}
  MaterialExpressionConstant_52 {class=MaterialExpressionConstant, properties {r=-2}}
  MaterialExpressionConstant_53 {class=MaterialExpressionConstant, properties {r=1}}
  MaterialExpressionConstant_6 {class=MaterialExpressionConstant}
  MaterialExpressionConstant_7 {class=MaterialExpressionConstant, properties {r=0.3}}
  MaterialExpressionConstant_8 {class=MaterialExpressionConstant, properties {r=1}}
  MaterialExpressionConstant_9 {class=MaterialExpressionConstant, properties {r=0.7}}
  MaterialExpressionConvert_10 {
   class=MaterialExpressionConvert
   inputs [
    None {from=MaterialExpressionMultiply_178}
    None {from=MaterialExpressionMultiply_177}
   ]
   properties {
    nodeName=Make Float2
    convertInputs [
     {
      type=Scalar
      expressionInput {
       expression=&self:MaterialExpressionMultiply_178
       inputName=None
       mask=0
       maskA=0
       maskB=0
       maskG=0
       maskR=0
       outputIndex=0
      }
     }
     {
      type=Scalar
      expressionInput {
       expression=&self:MaterialExpressionMultiply_177
       inputName=None
       mask=0
       maskA=0
       maskB=0
       maskG=0
       maskR=0
       outputIndex=0
      }
     }
    ]
    convertOutputs [{type=Vector2}]
   }
  }
  MaterialExpressionConvert_16 {
   class=MaterialExpressionConvert
   inputs [
    None {from=MaterialExpressionMultiply_180}
    None {from=MaterialExpressionMultiply_179}
   ]
   properties {
    nodeName=Make Float2
    convertInputs [
     {
      type=Scalar
      expressionInput {
       expression=&self:MaterialExpressionMultiply_180
       inputName=None
       mask=0
       maskA=0
       maskB=0
       maskG=0
       maskR=0
       outputIndex=0
      }
     }
     {
      type=Scalar
      expressionInput {
       expression=&self:MaterialExpressionMultiply_179
       inputName=None
       mask=0
       maskA=0
       maskB=0
       maskG=0
       maskR=0
       outputIndex=0
      }
     }
    ]
    convertOutputs [{type=Vector2}]
   }
  }
  MaterialExpressionConvert_18 {
   class=MaterialExpressionConvert
   inputs [
    None {from=MaterialExpressionMultiply_184}
    None {from=MaterialExpressionMultiply_183}
   ]
   properties {
    nodeName=Make Float2
    convertInputs [
     {
      type=Scalar
      expressionInput {
       expression=&self:MaterialExpressionMultiply_184
       inputName=None
       mask=0
       maskA=0
       maskB=0
       maskG=0
       maskR=0
       outputIndex=0
      }
     }
     {
      type=Scalar
      expressionInput {
       expression=&self:MaterialExpressionMultiply_183
       inputName=None
       mask=0
       maskA=0
       maskB=0
       maskG=0
       maskR=0
       outputIndex=0
      }
     }
    ]
    convertOutputs [{type=Vector2}]
   }
  }
  MaterialExpressionConvert_2 {
   class=MaterialExpressionConvert
   inputs [
    None {from=MaterialExpressionAdd_43}
    None {from=MaterialExpressionAdd_42}
   ]
   properties {
    nodeName=Make Float2
    convertInputs [
     {
      type=Scalar
      expressionInput {
       expression=&self:MaterialExpressionAdd_43
       inputName=None
       mask=0
       maskA=0
       maskB=0
       maskG=0
       maskR=0
       outputIndex=0
      }
     }
     {
      type=Scalar
      expressionInput {
       expression=&self:MaterialExpressionAdd_42
       inputName=None
       mask=0
       maskA=0
       maskB=0
       maskG=0
       maskR=0
       outputIndex=0
      }
     }
    ]
    convertOutputs [{type=Vector2}]
   }
  }
  MaterialExpressionConvert_4 {
   class=MaterialExpressionConvert
   inputs [
    None {from=MaterialExpressionAdd_48}
    None {from=MaterialExpressionAdd_47}
   ]
   properties {
    nodeName=Make Float2
    convertInputs [
     {
      type=Scalar
      expressionInput {
       expression=&self:MaterialExpressionAdd_48
       inputName=None
       mask=0
       maskA=0
       maskB=0
       maskG=0
       maskR=0
       outputIndex=0
      }
     }
     {
      type=Scalar
      expressionInput {
       expression=&self:MaterialExpressionAdd_47
       inputName=None
       mask=0
       maskA=0
       maskB=0
       maskG=0
       maskR=0
       outputIndex=0
      }
     }
    ]
    convertOutputs [{type=Vector2}]
   }
  }
  MaterialExpressionConvert_5 {
   class=MaterialExpressionConvert
   inputs [None {from=MaterialExpressionStaticSwitchParameter_33}]
   properties {
    nodeName=Break Float3
    convertInputs [{
      type=Vector3
      expressionInput {
       expression=&self:MaterialExpressionStaticSwitchParameter_33
       inputName=None
       mask=0
       maskA=0
       maskB=0
       maskG=0
       maskR=0
       outputIndex=0
      }
     }]
    convertOutputs [{type=Scalar}, {type=Scalar}, {type=Scalar}]
   }
  }
  MaterialExpressionConvert_6 {
   class=MaterialExpressionConvert
   inputs [None {from=MaterialExpressionMultiply_3}]
   properties {
    nodeName=Break Float3
    convertInputs [{
      type=Vector3
      expressionInput {
       expression=&self:MaterialExpressionMultiply_3
       inputName=None
       mask=0
       maskA=0
       maskB=0
       maskG=0
       maskR=0
       outputIndex=0
      }
     }]
    convertOutputs [{type=Scalar}, {type=Scalar}, {type=Scalar}]
   }
  }
  MaterialExpressionConvert_9 {
   class=MaterialExpressionConvert
   inputs [
    None {from=MaterialExpressionMultiply_176}
    None {from=MaterialExpressionMultiply_175}
   ]
   properties {
    nodeName=Make Float2
    convertInputs [
     {
      type=Scalar
      expressionInput {
       expression=&self:MaterialExpressionMultiply_176
       inputName=None
       mask=0
       maskA=0
       maskB=0
       maskG=0
       maskR=0
       outputIndex=0
      }
     }
     {
      type=Scalar
      expressionInput {
       expression=&self:MaterialExpressionMultiply_175
       inputName=None
       mask=0
       maskA=0
       maskB=0
       maskG=0
       maskR=0
       outputIndex=0
      }
     }
    ]
    convertOutputs [{type=Vector2}]
   }
  }
  MaterialExpressionDeriveNormalZ_1 {class=MaterialExpressionDeriveNormalZ, inputs [InXY {from=MaterialExpressionMultiply_108, output=""}], properties {}}
  MaterialExpressionDesaturation_1 {
   class=MaterialExpressionDesaturation
   inputs [
    None {from=MaterialExpressionVectorParameter_5, output=RGB}
    Fraction {from=MaterialExpressionConstant_0, output=""}
   ]
   properties {}
  }
  MaterialExpressionDesaturation_5 {
   class=MaterialExpressionDesaturation
   inputs [
    None {from=MaterialExpressionNamedRerouteUsage_89, output=""}
    Fraction {from="", output=""}
   ]
   properties {}
  }
  MaterialExpressionDesaturation_6 {
   class=MaterialExpressionDesaturation
   inputs [
    None {from=MaterialExpressionStaticSwitchParameter_9, output=""}
    Fraction {from=MaterialExpressionConstant_33, output=""}
   ]
   properties {}
  }
  MaterialExpressionDesaturation_7 {
   class=MaterialExpressionDesaturation
   inputs [
    None {from=MaterialExpressionNamedRerouteUsage_97, output=""}
    Fraction {from=MaterialExpressionConstant_52, output=""}
   ]
   properties {}
  }
  MaterialExpressionDistance_0 {
   class=MaterialExpressionDistance
   inputs [
    A {from=MaterialExpressionAdd_44, output=""}
    B {from=MaterialExpressionConstant_32, output=""}
   ]
   properties {}
  }
  MaterialExpressionDivide_0 {
   class=MaterialExpressionDivide
   inputs [
    A {from=MaterialExpressionPixelDepth_0, output=R}
    B {from=MaterialExpressionMultiply_12}
   ]
   properties {constB=350}
  }
  MaterialExpressionDivide_1 {
   class=MaterialExpressionDivide
   inputs [
    A {from=MaterialExpressionTime_3}
    B {from=MaterialExpressionScalarParameter_4}
   ]
   properties {constB=5}
  }
  MaterialExpressionDivide_11 {
   class=MaterialExpressionDivide
   inputs [
    A {from=MaterialExpressionNamedRerouteUsage_10}
    B {from=MaterialExpressionLinearInterpolate_25}
   ]
   properties {constB=40}
  }
  MaterialExpressionDivide_3 {
   class=MaterialExpressionDivide
   inputs [
    A {from=MaterialExpressionMultiply_11}
    B {from=MaterialExpressionDotProduct_3}
   ]
  }
  MaterialExpressionDivide_5 {
   class=MaterialExpressionDivide
   inputs [
    A {from=MaterialExpressionPixelDepth_2, output=R}
    B {from=MaterialExpressionScalarParameter_92}
   ]
   properties {constB=350}
  }
  MaterialExpressionDivide_6 {
   class=MaterialExpressionDivide
   inputs [
    A {from=MaterialExpressionTime_4}
    B {from=""}
   ]
  }
  MaterialExpressionDivide_9 {
   class=MaterialExpressionDivide
   inputs [
    A {from=MaterialExpressionPixelDepth_3, output=R}
    B {from=MaterialExpressionStaticSwitchParameter_36}
   ]
   properties {constB=350}
  }
  MaterialExpressionDotProduct_0 {
   class=MaterialExpressionDotProduct
   inputs [
    A {from=MaterialExpressionMultiply_29, output=""}
    B {from=MaterialExpressionVertexNormalWS_1, output=""}
   ]
   properties {}
  }
  MaterialExpressionDotProduct_3 {
   class=MaterialExpressionDotProduct
   inputs [
    A {from=MaterialExpressionReroute_3, output=""}
    B {from=MaterialExpressionVertexNormalWS_3, output=""}
   ]
   properties {}
  }
  MaterialExpressionEyeAdaptationInverse_1 {
   class=MaterialExpressionEyeAdaptationInverse
   inputs [
    LightValueInput {from=MaterialExpressionMultiply_17, output=""}
    AlphaInput {from=MaterialExpressionConstant_27, output=""}
   ]
   properties {}
  }
  MaterialExpressionFmod_0 {
   class=MaterialExpressionFmod
   inputs [
    A {from=MaterialExpressionMaterialFunctionCall_9, output=Result}
    B {from=MaterialExpressionConstant_17, output=""}
   ]
   properties {}
  }
  MaterialExpressionFmod_1 {
   class=MaterialExpressionFmod
   inputs [
    A {from=MaterialExpressionMaterialFunctionCall_18, output=Result}
    B {from=MaterialExpressionConstant_9, output=""}
   ]
   properties {}
  }
  MaterialExpressionFrac_0 {class=MaterialExpressionFrac, inputs [None {from=MaterialExpressionMultiply_109, output=""}], properties {}}
  MaterialExpressionFrac_3 {class=MaterialExpressionFrac, inputs [None {from=MaterialExpressionMultiply_114, output=""}], properties {}}
  MaterialExpressionFunctionInput_2 {
   class=MaterialExpressionFunctionInput
   inputs [Preview {from="", output=""}]
   properties {bUsePreviewValueAsDefault=true, inputName=SFX, sortPriority=3}
  }
  MaterialExpressionFunctionInput_3 {
   class=MaterialExpressionFunctionInput
   inputs [Preview {from="", output=""}]
   properties {inputName=Original MA, inputType=FunctionInput_MaterialAttributes}
  }
  MaterialExpressionFunctionOutput_0 {
   class=MaterialExpressionFunctionOutput
   inputs [None {from=MaterialExpressionStaticSwitchParameter_16, output=""}]
   properties {outputName=(MA)}
  }
  MaterialExpressionGetMaterialAttributes_1 {
   class=MaterialExpressionGetMaterialAttributes
   inputs [None {from=MaterialExpressionNamedRerouteUsage_125, output=""}]
   properties {attributeGetTypes [
     #1
     #2
     #3
     #4
     #5
     #6
     #7
    ]}
  }
  MaterialExpressionGetMaterialAttributes_3 {
   class=MaterialExpressionGetMaterialAttributes
   inputs [None {from=MaterialExpressionNamedRerouteUsage_124, output=""}]
   properties {attributeGetTypes [
     #1
     #2
     #3
     #4
     #5
     #6
     #7
    ]}
  }
  MaterialExpressionLinearInterpolate_0 {
   class=MaterialExpressionLinearInterpolate
   inputs [
    A {from=""}
    B {from=""}
    Alpha {from=MaterialExpressionNamedRerouteUsage_1}
   ]
   properties {constA=1, constB=1.3}
  }
  MaterialExpressionLinearInterpolate_1 {
   class=MaterialExpressionLinearInterpolate
   inputs [
    A {from=MaterialExpressionLinearInterpolate_43}
    B {from=MaterialExpressionNamedRerouteUsage_13}
    Alpha {from=MaterialExpressionSaturate_0}
   ]
  }
  MaterialExpressionLinearInterpolate_10 {
   class=MaterialExpressionLinearInterpolate
   inputs [
    A {from=""}
    B {from=""}
    Alpha {from=MaterialExpressionNamedRerouteUsage_169}
   ]
   properties {constA=0.85, constB=0.4}
  }
  MaterialExpressionLinearInterpolate_11 {
   class=MaterialExpressionLinearInterpolate
   inputs [
    A {from=MaterialExpressionNamedRerouteUsage_63}
    B {from=MaterialExpressionLinearInterpolate_18}
    Alpha {from=MaterialExpressionNamedRerouteUsage_61}
   ]
  }
  MaterialExpressionLinearInterpolate_12 {
   class=MaterialExpressionLinearInterpolate
   inputs [
    A {from=""}
    B {from=MaterialExpressionMultiply_18}
    Alpha {from=MaterialExpressionNamedRerouteUsage_14}
   ]
   properties {constA=1}
  }
  MaterialExpressionLinearInterpolate_13 {
   class=MaterialExpressionLinearInterpolate
   inputs [
    A {from=MaterialExpressionNamedRerouteUsage_46}
    B {from=MaterialExpressionNamedRerouteUsage_43}
    Alpha {from=MaterialExpressionNamedRerouteUsage_32}
   ]
  }
  MaterialExpressionLinearInterpolate_14 {
   class=MaterialExpressionLinearInterpolate
   inputs [
    A {from=MaterialExpressionNamedRerouteUsage_45}
    B {from=MaterialExpressionNamedRerouteUsage_44}
    Alpha {from=MaterialExpressionNamedRerouteUsage_29}
   ]
  }
  MaterialExpressionLinearInterpolate_15 {
   class=MaterialExpressionLinearInterpolate
   inputs [
    A {from=MaterialExpressionNamedRerouteUsage_27}
    B {from=MaterialExpressionNamedRerouteUsage_40}
    Alpha {from=MaterialExpressionNamedRerouteUsage_30}
   ]
  }
  MaterialExpressionLinearInterpolate_16 {
   class=MaterialExpressionLinearInterpolate
   inputs [
    A {from=MaterialExpressionNamedRerouteUsage_47}
    B {from=MaterialExpressionNamedRerouteUsage_41}
    Alpha {from=MaterialExpressionNamedRerouteUsage_31}
   ]
  }
  MaterialExpressionLinearInterpolate_17 {
   class=MaterialExpressionLinearInterpolate
   inputs [
    A {from=MaterialExpressionLinearInterpolate_7}
    B {from=MaterialExpressionReroute_8}
    Alpha {from=MaterialExpressionNamedRerouteUsage_26}
   ]
  }
  MaterialExpressionLinearInterpolate_18 {
   class=MaterialExpressionLinearInterpolate
   inputs [
    A {from=MaterialExpressionReroute_11}
    B {from=MaterialExpressionMultiply_33}
    Alpha {from=MaterialExpressionNamedRerouteUsage_35}
   ]
  }
  MaterialExpressionLinearInterpolate_19 {
   class=MaterialExpressionLinearInterpolate
   inputs [
    A {from=MaterialExpressionMultiply_31}
    B {from=MaterialExpressionMultiply_33}
    Alpha {from=MaterialExpressionNamedRerouteUsage_36}
   ]
  }
  MaterialExpressionLinearInterpolate_2 {
   class=MaterialExpressionLinearInterpolate
   inputs [
    A {from=""}
    B {from=""}
    Alpha {from=MaterialExpressionNamedRerouteUsage_15}
   ]
   properties {constA=0.6, constB=3}
  }
  MaterialExpressionLinearInterpolate_20 {
   class=MaterialExpressionLinearInterpolate
   inputs [
    A {from=MaterialExpressionNamedRerouteUsage_48}
    B {from=MaterialExpressionAdd_13}
    Alpha {from=MaterialExpressionNamedRerouteUsage_37}
   ]
  }
  MaterialExpressionLinearInterpolate_21 {
   class=MaterialExpressionLinearInterpolate
   inputs [
    A {from=MaterialExpressionLinearInterpolate_5}
    B {from=MaterialExpressionLinearInterpolate_3}
    Alpha {from=MaterialExpressionNamedRerouteUsage_38}
   ]
  }
  MaterialExpressionLinearInterpolate_22 {
   class=MaterialExpressionLinearInterpolate
   inputs [
    A {from=MaterialExpressionNamedRerouteUsage_69}
    B {from=MaterialExpressionLinearInterpolate_19}
    Alpha {from=MaterialExpressionNamedRerouteUsage_64}
   ]
  }
  MaterialExpressionLinearInterpolate_23 {
   class=MaterialExpressionLinearInterpolate
   inputs [
    A {from=MaterialExpressionNamedRerouteUsage_41}
    B {from=MaterialExpressionLinearInterpolate_16}
    Alpha {from=MaterialExpressionNamedRerouteUsage_66}
   ]
  }
  MaterialExpressionLinearInterpolate_24 {
   class=MaterialExpressionLinearInterpolate
   inputs [
    A {from=MaterialExpressionNamedRerouteUsage_40}
    B {from=MaterialExpressionLinearInterpolate_15}
    Alpha {from=MaterialExpressionNamedRerouteUsage_68}
   ]
  }
  MaterialExpressionLinearInterpolate_25 {
   class=MaterialExpressionLinearInterpolate
   inputs [
    A {from=""}
    B {from=""}
    Alpha {from=MaterialExpressionNamedRerouteUsage_22}
   ]
   properties {constA=80, constB=79.8}
  }
  MaterialExpressionLinearInterpolate_26 {
   class=MaterialExpressionLinearInterpolate
   inputs [
    A {from=MaterialExpressionVectorParameter_0, output=RGB}
    B {from=MaterialExpressionVectorParameter_12, output=RGB}
    Alpha {from=MaterialExpressionNamedRerouteUsage_65}
   ]
  }
  MaterialExpressionLinearInterpolate_28 {
   class=MaterialExpressionLinearInterpolate
   inputs [
    A {from=MaterialExpressionMaterialFunctionCall_14, output=Glitched UV}
    B {from=MaterialExpressionAdd_18}
    Alpha {from=MaterialExpressionNamedRerouteUsage_67}
   ]
  }
  MaterialExpressionLinearInterpolate_29 {
   class=MaterialExpressionLinearInterpolate
   inputs [
    A {from=MaterialExpressionVectorParameter_21, output=RGB}
    B {from=MaterialExpressionVectorParameter_12, output=RGB}
    Alpha {from=MaterialExpressionNamedRerouteUsage_71}
   ]
  }
  MaterialExpressionLinearInterpolate_3 {
   class=MaterialExpressionLinearInterpolate
   inputs [
    A {from=MaterialExpressionNamedRerouteUsage_53}
    B {from=MaterialExpressionAdd_11}
    Alpha {from=MaterialExpressionNamedRerouteUsage_49}
   ]
  }
  MaterialExpressionLinearInterpolate_38 {
   class=MaterialExpressionLinearInterpolate
   inputs [
    A {from=MaterialExpressionReroute_9}
    B {from=MaterialExpressionOneMinus_6}
    Alpha {from=MaterialExpressionReroute_26}
   ]
  }
  MaterialExpressionLinearInterpolate_39 {
   class=MaterialExpressionLinearInterpolate
   inputs [
    A {from=""}
    B {from=""}
    Alpha {from=MaterialExpressionSaturate_36}
   ]
   properties {constA=0.1, constB=0.9}
  }
  MaterialExpressionLinearInterpolate_4 {
   class=MaterialExpressionLinearInterpolate
   inputs [
    A {from=MaterialExpressionTextureSampleParameter2D_0, output=R}
    B {from=MaterialExpressionMultiply_26}
    Alpha {from=MaterialExpressionSaturate_5}
   ]
  }
  MaterialExpressionLinearInterpolate_42 {
   class=MaterialExpressionLinearInterpolate
   inputs [
    A {from=MaterialExpressionAdd_71}
    B {from=MaterialExpressionReroute_39}
    Alpha {from=MaterialExpressionReroute_38}
   ]
  }
  MaterialExpressionLinearInterpolate_43 {
   class=MaterialExpressionLinearInterpolate
   inputs [
    A {from=MaterialExpressionMaterialFunctionCall_44, output=Result}
    B {from=MaterialExpressionAdd_4}
    Alpha {from=MaterialExpressionNamedRerouteUsage_91}
   ]
   properties {constB=0}
  }
  MaterialExpressionLinearInterpolate_44 {
   class=MaterialExpressionLinearInterpolate
   inputs [
    A {from=MaterialExpressionMaterialFunctionCall_46, output=Result}
    B {from=MaterialExpressionLinearInterpolate_1}
    Alpha {from=MaterialExpressionSaturate_62}
   ]
   properties {constAlpha=0.9}
  }
  MaterialExpressionLinearInterpolate_45 {
   class=MaterialExpressionLinearInterpolate
   inputs [
    A {from=MaterialExpressionMultiply_98}
    B {from=MaterialExpressionNamedRerouteUsage_11}
    Alpha {from=MaterialExpressionNamedRerouteUsage_108}
   ]
   properties {constB=0}
  }
  MaterialExpressionLinearInterpolate_46 {
   class=MaterialExpressionLinearInterpolate
   inputs [
    A {from=MaterialExpressionReroute_34}
    B {from=MaterialExpressionAdd_23}
    Alpha {from=MaterialExpressionSaturate_37}
   ]
  }
  MaterialExpressionLinearInterpolate_47 {
   class=MaterialExpressionLinearInterpolate
   inputs [
    A {from=MaterialExpressionMultiply_186}
    B {from=MaterialExpressionMultiply_111}
    Alpha {from=MaterialExpressionSaturate_71}
   ]
  }
  MaterialExpressionLinearInterpolate_49 {
   class=MaterialExpressionLinearInterpolate
   inputs [
    A {from=""}
    B {from=""}
    Alpha {from=MaterialExpressionNamedRerouteUsage_98}
   ]
   properties {constA=0.6, constB=0.3}
  }
  MaterialExpressionLinearInterpolate_5 {
   class=MaterialExpressionLinearInterpolate
   inputs [
    A {from=MaterialExpressionMaterialFunctionCall_25, output=Glitched UV}
    B {from=MaterialExpressionLinearInterpolate_3}
    Alpha {from=""}
   ]
  }
  MaterialExpressionLinearInterpolate_52 {
   class=MaterialExpressionLinearInterpolate
   inputs [
    A {from=MaterialExpressionReroute_47}
    B {from=MaterialExpressionSubtract_0}
    Alpha {from=MaterialExpressionNamedRerouteUsage_109}
   ]
  }
  MaterialExpressionLinearInterpolate_53 {
   class=MaterialExpressionLinearInterpolate
   inputs [
    A {from=MaterialExpressionMultiply_133}
    B {from=MaterialExpressionMultiply_127}
    Alpha {from=MaterialExpressionSaturate_42}
   ]
  }
  MaterialExpressionLinearInterpolate_54 {
   class=MaterialExpressionLinearInterpolate
   inputs [
    A {from=MaterialExpressionStaticSwitchParameter_7}
    B {from=MaterialExpressionStaticSwitchParameter_8}
    Alpha {from=MaterialExpressionSaturate_33}
   ]
  }
  MaterialExpressionLinearInterpolate_55 {
   class=MaterialExpressionLinearInterpolate
   inputs [
    A {from=MaterialExpressionMultiply_137}
    B {from=MaterialExpressionMultiply_136}
    Alpha {from=MaterialExpressionSaturate_33}
   ]
  }
  MaterialExpressionLinearInterpolate_56 {
   class=MaterialExpressionLinearInterpolate
   inputs [
    A {from=MaterialExpressionReroute_37}
    B {from=MaterialExpressionNamedRerouteUsage_112}
    Alpha {from=MaterialExpressionNamedRerouteUsage_114}
   ]
  }
  MaterialExpressionLinearInterpolate_57 {
   class=MaterialExpressionLinearInterpolate
   inputs [
    A {from=MaterialExpressionLinearInterpolate_56}
    B {from=MaterialExpressionReroute_40}
    Alpha {from=MaterialExpressionMultiply_155}
   ]
  }
  MaterialExpressionLinearInterpolate_58 {
   class=MaterialExpressionLinearInterpolate
   inputs [
    A {from=MaterialExpressionConstant3Vector_0}
    B {from=MaterialExpressionLinearInterpolate_57}
    Alpha {from=MaterialExpressionSaturate_50}
   ]
  }
  MaterialExpressionLinearInterpolate_59 {
   class=MaterialExpressionLinearInterpolate
   inputs [
    A {from=MaterialExpressionScalarParameter_68}
    B {from=MaterialExpressionScalarParameter_72}
    Alpha {from=MaterialExpressionSaturate_51}
   ]
   properties {constB=0.05}
  }
  MaterialExpressionLinearInterpolate_6 {
   class=MaterialExpressionLinearInterpolate
   inputs [
    A {from=""}
    B {from=MaterialExpressionScalarParameter_7}
    Alpha {from=MaterialExpressionSubtract_2}
   ]
  }
  MaterialExpressionLinearInterpolate_60 {
   class=MaterialExpressionLinearInterpolate
   inputs [
    A {from=MaterialExpressionScalarParameter_66}
    B {from=MaterialExpressionScalarParameter_70}
    Alpha {from=MaterialExpressionSaturate_51}
   ]
  }
  MaterialExpressionLinearInterpolate_61 {
   class=MaterialExpressionLinearInterpolate
   inputs [
    A {from=""}
    B {from=MaterialExpressionMultiply_140}
    Alpha {from=MaterialExpressionSaturate_52}
   ]
   properties {constA=1}
  }
  MaterialExpressionLinearInterpolate_62 {
   class=MaterialExpressionLinearInterpolate
   inputs [
    A {from=MaterialExpressionNamedRerouteUsage_121}
    B {from=MaterialExpressionNamedRerouteUsage_120}
    Alpha {from=MaterialExpressionSaturate_55}
   ]
  }
  MaterialExpressionLinearInterpolate_63 {
   class=MaterialExpressionLinearInterpolate
   inputs [
    A {from=MaterialExpressionTextureCoordinate_7}
    B {from=MaterialExpressionNamedRerouteDeclaration_47}
    Alpha {from=MaterialExpressionSaturate_60}
   ]
  }
  MaterialExpressionLinearInterpolate_64 {
   class=MaterialExpressionLinearInterpolate
   inputs [
    A {from=MaterialExpressionTextureCoordinate_9}
    B {from=MaterialExpressionNamedRerouteUsage_127}
    Alpha {from=MaterialExpressionNamedRerouteUsage_5}
   ]
  }
  MaterialExpressionLinearInterpolate_65 {
   class=MaterialExpressionLinearInterpolate
   inputs [
    A {from=MaterialExpressionMultiply_139}
    B {from=MaterialExpressionScalarParameter_66}
    Alpha {from=MaterialExpressionNamedRerouteUsage_128}
   ]
  }
  MaterialExpressionLinearInterpolate_66 {
   class=MaterialExpressionLinearInterpolate
   inputs [
    A {from=MaterialExpressionLinearInterpolate_59}
    B {from=MaterialExpressionScalarParameter_68}
    Alpha {from=MaterialExpressionNamedRerouteUsage_129}
   ]
  }
  MaterialExpressionLinearInterpolate_67 {
   class=MaterialExpressionLinearInterpolate
   inputs [
    A {from=MaterialExpressionAdd_1}
    B {from=MaterialExpressionScalarParameter_41}
    Alpha {from=MaterialExpressionNamedRerouteUsage_169}
   ]
   properties {constA=0.92}
  }
  MaterialExpressionLinearInterpolate_68 {
   class=MaterialExpressionLinearInterpolate
   inputs [
    A {from=""}
    B {from=MaterialExpressionSaturate_64}
    Alpha {from=MaterialExpressionSaturate_65}
   ]
   properties {constA=1}
  }
  MaterialExpressionLinearInterpolate_7 {
   class=MaterialExpressionLinearInterpolate
   inputs [
    A {from=MaterialExpressionLinearInterpolate_20}
    B {from=MaterialExpressionMaterialFunctionCall_22, output=Glitched UV}
    Alpha {from=""}
   ]
  }
  MaterialExpressionLinearInterpolate_73 {
   class=MaterialExpressionLinearInterpolate
   inputs [
    A {from=MaterialExpressionNamedRerouteUsage_140}
    B {from=MaterialExpressionNamedRerouteUsage_141}
    Alpha {from=MaterialExpressionNamedRerouteDeclaration_4}
   ]
  }
  MaterialExpressionLinearInterpolate_74 {
   class=MaterialExpressionLinearInterpolate
   inputs [
    A {from=MaterialExpressionComponentMask_0}
    B {from=""}
    Alpha {from=MaterialExpressionSaturate_69}
   ]
  }
  MaterialExpressionLinearInterpolate_76 {
   class=MaterialExpressionLinearInterpolate
   inputs [
    A {from=MaterialExpressionSaturate_53}
    B {from=MaterialExpressionSaturate_54}
    Alpha {from=MaterialExpressionNamedRerouteUsage_143}
   ]
  }
  MaterialExpressionLinearInterpolate_77 {
   class=MaterialExpressionLinearInterpolate
   inputs [
    A {from=MaterialExpressionSaturate_54}
    B {from=MaterialExpressionMaterialFunctionCall_49, output=Result}
    Alpha {from=MaterialExpressionNamedRerouteUsage_144}
   ]
  }
  MaterialExpressionLinearInterpolate_78 {
   class=MaterialExpressionLinearInterpolate
   inputs [
    A {from=MaterialExpressionLinearInterpolate_80}
    B {from=MaterialExpressionSubtract_1}
    Alpha {from=MaterialExpressionNamedRerouteUsage_181}
   ]
  }
  MaterialExpressionLinearInterpolate_8 {
   class=MaterialExpressionLinearInterpolate
   inputs [
    A {from=""}
    B {from=MaterialExpressionScalarParameter_6}
    Alpha {from=MaterialExpressionNamedRerouteUsage_2}
   ]
  }
  MaterialExpressionLinearInterpolate_80 {
   class=MaterialExpressionLinearInterpolate
   inputs [
    A {from=MaterialExpressionAdd_59}
    B {from=""}
    Alpha {from=MaterialExpressionMultiply_206}
   ]
   properties {constB=0.9}
  }
  MaterialExpressionLinearInterpolate_9 {
   class=MaterialExpressionLinearInterpolate
   inputs [
    A {from=MaterialExpressionLinearInterpolate_10}
    B {from=MaterialExpressionLinearInterpolate_67}
    Alpha {from=MaterialExpressionNamedRerouteUsage_17}
   ]
   properties {constA=0.5}
  }
  MaterialExpressionMakeMaterialAttributes_1 {
   class=MaterialExpressionMakeMaterialAttributes
   inputs [
    BaseColor {from=MaterialExpressionStaticSwitchParameter_1, output=""}
    Metallic {from=MaterialExpressionConstant_35, output=""}
    Specular {from=MaterialExpressionStaticSwitchParameter_27, output=""}
    Roughness {from=MaterialExpressionStaticSwitchParameter_28, output=""}
    Anisotropy {from="", output=""}
    EmissiveColor {from=MaterialExpressionNamedRerouteUsage_130, output=""}
    Opacity {from="", output=""}
    OpacityMask {from="", output=""}
    Normal {from=MaterialExpressionNamedRerouteUsage_21, output=""}
    Tangent {from="", output=""}
    WorldPositionOffset {from="", output=""}
    SubsurfaceColor {from="", output=""}
    ClearCoat {from="", output=""}
    ClearCoatRoughness {from="", output=""}
    AmbientOcclusion {from="", output=""}
    Refraction {from="", output=""}
    CustomizedUVs_0 {from="", output=""}
    CustomizedUVs_1 {from="", output=""}
    CustomizedUVs_2 {from="", output=""}
    CustomizedUVs_3 {from="", output=""}
    CustomizedUVs_4 {from="", output=""}
    CustomizedUVs_5 {from="", output=""}
    CustomizedUVs_6 {from="", output=""}
    CustomizedUVs_7 {from="", output=""}
    PixelDepthOffset {from="", output=""}
    ShadingModel {from="", output=""}
    Displacement {from="", output=""}
   ]
   properties {}
  }
  MaterialExpressionMaterialFunctionCall_0 {
   class=MaterialExpressionMaterialFunctionCall
   inputs [
    BaseNormal {from=MaterialExpressionGetMaterialAttributes_3, output=Normal}
    AdditionalNormal {from=MaterialExpressionGetMaterialAttributes_1, output=Normal}
   ]
   properties {materialFunction=&BlendAngleCorrectedNormals}
  }
  MaterialExpressionMaterialFunctionCall_1 {
   class=MaterialExpressionMaterialFunctionCall
   inputs [
    In {from=MaterialExpressionNamedRerouteUsage_91}
    Contrast {from=MaterialExpressionScalarParameter_19}
   ]
   properties {materialFunction=&CheapContrast}
  }
  MaterialExpressionMaterialFunctionCall_10 {
   class=MaterialExpressionMaterialFunctionCall
   inputs [
    In {from=MaterialExpressionMultiply_24}
    Contrast {from=MaterialExpressionConstant_15}
   ]
   properties {materialFunction=&CheapContrast}
  }
  MaterialExpressionMaterialFunctionCall_11 {
   class=MaterialExpressionMaterialFunctionCall
   inputs [
    UVs {from=""}
    Rotation Center {from=""}
    "Rotation Angle (0-1)" {from=""}
   ]
   properties {materialFunction=&CustomRotator}
  }
  MaterialExpressionMaterialFunctionCall_12 {
   class=MaterialExpressionMaterialFunctionCall
   inputs [
    "Xb1,PS4" {from=MaterialExpressionTextureSample_8, output=RGB}
    "PC-High&&Epic,XSX,PS5" {from=MaterialExpressionTextureSample_8, output=RGB}
    "PC-Perf,Android" {from=MaterialExpressionConstant_4}
    Switch {from=MaterialExpressionConstant_4}
    "PC-Low(NotPerf)&&Medium" {from=MaterialExpressionTextureSample_8, output=RGB}
   ]
   properties {materialFunction=/Game/Effects/Niagara/MasterMaterials/Functions/Common/PlatformSwitch_3D}
  }
  MaterialExpressionMaterialFunctionCall_13 {
   class=MaterialExpressionMaterialFunctionCall
   inputs [Float3 {from=MaterialExpressionMaterialFunctionCall_12, output=Result}]
   properties {materialFunction=&BreakOutFloat3Components}
  }
  MaterialExpressionMaterialFunctionCall_14 {
   class=MaterialExpressionMaterialFunctionCall
   inputs [
    Time {from=MaterialExpressionTime_0}
    Glitch Flickering {from=MaterialExpressionScalarParameter_23}
    GlitchDuration {from=MaterialExpressionScalarParameter_20}
    GlitchFrequency {from=MaterialExpressionScalarParameter_21}
    GlitchMaskResolution {from=""}
    PixellateGlitch_MinResolution {from=""}
    PixellateGlitch_MaxResolution {from=""}
    UV {from=MaterialExpressionStaticSwitchParameter_31}
   ]
   properties {materialFunction=/Hermes/Techart/UtopiaEmissiveSurfaces/Glass_4Layer/Glitching/MF_MMGlitchingScreens}
  }
  MaterialExpressionMaterialFunctionCall_15 {
   class=MaterialExpressionMaterialFunctionCall
   inputs [
    "Xb1,PS4" {from=MaterialExpressionStaticSwitchParameter_34}
    "PC-High&&Epic,XSX,PS5" {from=MaterialExpressionStaticSwitchParameter_34}
    "PC-Perf,Android" {from=MaterialExpressionConstant_6}
    Switch {from=MaterialExpressionConstant_6}
    "PC-Low(NotPerf)&&Medium" {from=MaterialExpressionStaticSwitchParameter_34}
   ]
   properties {materialFunction=&PlatformSwitch_3D}
  }
  MaterialExpressionMaterialFunctionCall_17 {
   class=MaterialExpressionMaterialFunctionCall
   inputs [
    Normal {from=MaterialExpressionMaterialFunctionCall_24, output=Result}
    Flatness {from=MaterialExpressionSaturate_4}
   ]
   properties {materialFunction=&FlattenNormal}
  }
  MaterialExpressionMaterialFunctionCall_18 {
   class=MaterialExpressionMaterialFunctionCall
   inputs [Speed {from=MaterialExpressionConstant_7}]
   properties {materialFunction=&DebugTimeSine}
  }
  MaterialExpressionMaterialFunctionCall_19 {
   class=MaterialExpressionMaterialFunctionCall
   inputs [Float2 {from=MaterialExpressionNamedRerouteUsage_96}]
   properties {materialFunction=&BreakOutFloat2Components}
  }
  MaterialExpressionMaterialFunctionCall_2 {
   class=MaterialExpressionMaterialFunctionCall
   inputs [World Position {from=""}]
   properties {materialFunction=/Game/ArtTools/Tech_Art/DirectFX/Assets/MaterialFunctions/MF_ProjectionCoordinates}
  }
  MaterialExpressionMaterialFunctionCall_20 {
   class=MaterialExpressionMaterialFunctionCall
   inputs [
    Depth {from=MaterialExpressionScalarParameter_80}
    UVs {from=MaterialExpressionTextureCoordinate_2}
   ]
   properties {materialFunction=/Hermes/Techart/UtopiaEmissiveSurfaces/MF_PlanarBumpOffset}
  }
  MaterialExpressionMaterialFunctionCall_22 {
   class=MaterialExpressionMaterialFunctionCall
   inputs [
    Time {from=MaterialExpressionDivide_11}
    GlitchMaskResolution {from=""}
    PixellateGlitch_MinResolution {from=""}
    PixellateGlitch_MaxResolution {from=""}
    UV {from=MaterialExpressionNamedRerouteUsage_8}
   ]
   properties {materialFunction=/Hermes/Techart/UtopiaEmissiveSurfaces/MF_ASyncGlitch}
  }
  MaterialExpressionMaterialFunctionCall_23 {
   class=MaterialExpressionMaterialFunctionCall
   inputs [Float2 {from=MaterialExpressionLinearInterpolate_64}]
   properties {materialFunction=&BreakOutFloat2Components}
  }
  MaterialExpressionMaterialFunctionCall_24 {
   class=MaterialExpressionMaterialFunctionCall
   inputs [
    InNormal {from=MaterialExpressionLinearInterpolate_58}
    Strength {from=MaterialExpressionScalarParameter_65}
   ]
   properties {materialFunction=/Game/Effects/Fort_Effects/FrontEnd/S33/GateContent/SilenceMoon_NormalStrength}
  }
  MaterialExpressionMaterialFunctionCall_25 {
   class=MaterialExpressionMaterialFunctionCall
   inputs [
    Time {from=MaterialExpressionDivide_11}
    GlitchMaskResolution {from=""}
    PixellateGlitch_MinResolution {from=""}
    PixellateGlitch_MaxResolution {from=""}
    UV {from=MaterialExpressionNamedRerouteUsage_53}
   ]
   properties {materialFunction=&MF_ASyncGlitch}
  }
  MaterialExpressionMaterialFunctionCall_3 {
   class=MaterialExpressionMaterialFunctionCall
   inputs [Float4 {from=MaterialExpressionCollectionParameter_3}]
   properties {materialFunction=&BreakOutFloat4Components}
  }
  MaterialExpressionMaterialFunctionCall_31 {
   class=MaterialExpressionMaterialFunctionCall
   inputs [
    In {from=MaterialExpressionPower_3}
    Contrast {from=MaterialExpressionConstant_34}
   ]
   properties {materialFunction=&CheapContrast}
  }
  MaterialExpressionMaterialFunctionCall_32 {
   class=MaterialExpressionMaterialFunctionCall
   inputs [
    In {from=MaterialExpressionReroute_41}
    Contrast {from=MaterialExpressionConstant_37}
   ]
   properties {materialFunction=&CheapContrast}
  }
  MaterialExpressionMaterialFunctionCall_33 {
   class=MaterialExpressionMaterialFunctionCall
   inputs [
    In {from=MaterialExpressionFunctionInput_2}
    Contrast {from=MaterialExpressionConstant_39}
   ]
   properties {materialFunction=&CheapContrast}
  }
  MaterialExpressionMaterialFunctionCall_34 {
   class=MaterialExpressionMaterialFunctionCall
   inputs [
    In {from=MaterialExpressionMultiply_163}
    Contrast {from=MaterialExpressionScalarParameter_102}
   ]
   properties {materialFunction=/Engine/Functions/Engine_MaterialFunctions01/ImageAdjustment/CheapContrast_RGB}
  }
  MaterialExpressionMaterialFunctionCall_35 {
   class=MaterialExpressionMaterialFunctionCall
   inputs [
    "Normal (WS)" {from=""}
    Texture Scale {from=MaterialExpressionScalarParameter_86}
    Triplanar Fade {from=MaterialExpressionScalarParameter_87}
    Derivatives Scale {from=""}
   ]
   properties {materialFunction=/Game/Athena/Environments/MaterialFunctions/MF_TriplanarDithered}
  }
  MaterialExpressionMaterialFunctionCall_36 {
   class=MaterialExpressionMaterialFunctionCall
   inputs [World Position {from=MaterialExpressionAdd_19}]
   properties {materialFunction=&MF_ProjectionCoordinates}
  }
  MaterialExpressionMaterialFunctionCall_39 {
   class=MaterialExpressionMaterialFunctionCall
   inputs [Float2 {from=MaterialExpressionLinearInterpolate_62}]
   properties {materialFunction=&BreakOutFloat2Components}
  }
  MaterialExpressionMaterialFunctionCall_4 {
   class=MaterialExpressionMaterialFunctionCall
   inputs [World Position {from=""}]
   properties {materialFunction=&MF_ProjectionCoordinates}
  }
  MaterialExpressionMaterialFunctionCall_40 {
   class=MaterialExpressionMaterialFunctionCall
   inputs [Float2 {from=MaterialExpressionNamedRerouteUsage_104}]
   properties {materialFunction=&BreakOutFloat2Components}
  }
  MaterialExpressionMaterialFunctionCall_41 {
   class=MaterialExpressionMaterialFunctionCall
   inputs [Float2 {from=MaterialExpressionNamedRerouteUsage_95}]
   properties {materialFunction=&BreakOutFloat2Components}
  }
  MaterialExpressionMaterialFunctionCall_43 {
   class=MaterialExpressionMaterialFunctionCall
   inputs [Float2 {from=MaterialExpressionComponentMask_6}]
   properties {materialFunction=&BreakOutFloat2Components}
  }
  MaterialExpressionMaterialFunctionCall_44 {
   class=MaterialExpressionMaterialFunctionCall
   inputs [
    Blend {from=MaterialExpressionMultiply_187}
    Base {from=MaterialExpressionLinearInterpolate_45}
   ]
   properties {materialFunction=/Engine/Functions/Engine_MaterialFunctions03/Blends/Blend_Screen}
  }
  MaterialExpressionMaterialFunctionCall_46 {
   class=MaterialExpressionMaterialFunctionCall
   inputs [
    Blend {from=MaterialExpressionNamedRerouteUsage_167}
    Base {from=MaterialExpressionMultiply_2}
   ]
   properties {materialFunction=&Blend_Screen}
  }
  MaterialExpressionMaterialFunctionCall_47 {
   class=MaterialExpressionMaterialFunctionCall
   inputs [
    X {from=MaterialExpressionGetMaterialAttributes_3, output=Specular}
    Y {from=MaterialExpressionGetMaterialAttributes_3, output=Specular}
    Z {from=MaterialExpressionGetMaterialAttributes_3, output=Specular}
   ]
   properties {materialFunction=&MakeFloat3}
  }
  MaterialExpressionMaterialFunctionCall_49 {
   class=MaterialExpressionMaterialFunctionCall
   inputs [
    In {from=MaterialExpressionNamedRerouteUsage_170}
    Contrast {from=MaterialExpressionConstant_53}
   ]
   properties {materialFunction=&CheapContrast}
  }
  MaterialExpressionMaterialFunctionCall_5 {
   class=MaterialExpressionMaterialFunctionCall
   inputs [
    Pixel Amount {from=MaterialExpressionNamedRerouteUsage_174}
    Pixel Size {from=MaterialExpressionAdd_51}
    Pixel Smooth {from=MaterialExpressionScalarParameter_73}
    UVs {from=MaterialExpressionConvert_16}
   ]
   properties {materialFunction=/Hermes/Techart/UtopiaEmissiveSurfaces/MF_DotGrid}
  }
  MaterialExpressionMaterialFunctionCall_6 {
   class=MaterialExpressionMaterialFunctionCall
   inputs [
    In {from=MaterialExpressionReroute_5}
    Contrast {from=MaterialExpressionConstant_2}
   ]
   properties {materialFunction=&CheapContrast}
  }
  MaterialExpressionMaterialFunctionCall_7 {
   class=MaterialExpressionMaterialFunctionCall
   inputs [Float2 {from=MaterialExpressionMultiply_105}]
   properties {materialFunction=&BreakOutFloat2Components}
  }
  MaterialExpressionMaterialFunctionCall_8 {
   class=MaterialExpressionMaterialFunctionCall
   inputs [
    PixellateGlitch_MaxResolution {from=""}
    PixellateGlitch_MinResolution {from=""}
    GlitchFrequency {from=MaterialExpressionScalarParameter_15}
    UV {from=""}
    GlitchMaskResolution {from=""}
    Time {from=""}
    GlitchDuration {from=MaterialExpressionScalarParameter_14}
    Glitch Flickering {from=MaterialExpressionScalarParameter_16}
    DDX {from=MaterialExpressionNamedRerouteUsage_6}
    DDY {from=MaterialExpressionNamedRerouteUsage_7}
    Triplanar UV {from=MaterialExpressionNamedRerouteUsage_34}
    Glitch Size {from=MaterialExpressionScalarParameter_9}
    Color Glitch Speed {from=MaterialExpressionScalarParameter_11}
    Texture Glitch {from=""}
   ]
   properties {materialFunction=/Hermes/Techart/UtopiaEmissiveSurfaces/Glass_4Layer/Glitching/MF_GlitchTakeOver}
  }
  MaterialExpressionMaterialFunctionCall_9 {
   class=MaterialExpressionMaterialFunctionCall
   inputs [Speed {from=MaterialExpressionConstant_16}]
   properties {materialFunction=&DebugTimeSine}
  }
  MaterialExpressionMaterialProxyReplace_0 {
   class=MaterialExpressionMaterialProxyReplace
   inputs [
    Realtime {from=MaterialExpressionEyeAdaptationInverse_1, output=EyeAdaptationInverse}
    MaterialProxy {from=MaterialExpressionAdd_5, output=""}
   ]
   properties {}
  }
  MaterialExpressionMax_3 {
   class=MaterialExpressionMax
   inputs [
    A {from=MaterialExpressionMultiply_148}
    B {from=MaterialExpressionMultiply_149}
   ]
  }
  MaterialExpressionMax_4 {
   class=MaterialExpressionMax
   inputs [
    A {from=MaterialExpressionMultiply_150}
    B {from=MaterialExpressionMultiply_151}
   ]
  }
  MaterialExpressionMax_5 {
   class=MaterialExpressionMax
   inputs [
    A {from=MaterialExpressionMax_3}
    B {from=MaterialExpressionMax_4}
   ]
  }
  MaterialExpressionMultiply_0 {
   class=MaterialExpressionMultiply
   inputs [
    A {from=MaterialExpressionNamedRerouteUsage_0}
    B {from=""}
   ]
   properties {constB=0.5}
  }
  MaterialExpressionMultiply_1 {
   class=MaterialExpressionMultiply
   inputs [
    A {from=MaterialExpressionLinearInterpolate_0}
    B {from=MaterialExpressionLinearInterpolate_44}
   ]
  }
  MaterialExpressionMultiply_10 {
   class=MaterialExpressionMultiply
   inputs [
    A {from=MaterialExpressionGetMaterialAttributes_3, output=BaseColor}
    B {from=MaterialExpressionSaturate_63}
   ]
  }
  MaterialExpressionMultiply_101 {
   class=MaterialExpressionMultiply
   inputs [
    A {from=MaterialExpressionMultiply_134}
    B {from=MaterialExpressionMultiply_52}
   ]
  }
  MaterialExpressionMultiply_103 {
   class=MaterialExpressionMultiply
   inputs [
    A {from=MaterialExpressionStaticSwitchParameter_23}
    B {from=MaterialExpressionScalarParameter_12}
   ]
  }
  MaterialExpressionMultiply_104 {
   class=MaterialExpressionMultiply
   inputs [
    A {from=MaterialExpressionLinearInterpolate_47}
    B {from=MaterialExpressionNamedRerouteUsage_93}
   ]
  }
  MaterialExpressionMultiply_105 {
   class=MaterialExpressionMultiply
   inputs [
    A {from=MaterialExpressionAdd_30}
    B {from=MaterialExpressionAdd_30}
   ]
  }
  MaterialExpressionMultiply_106 {
   class=MaterialExpressionMultiply
   inputs [
    A {from=MaterialExpressionAdd_37}
    B {from=MaterialExpressionScalarParameter_25}
   ]
   properties {constB=90}
  }
  MaterialExpressionMultiply_107 {
   class=MaterialExpressionMultiply
   inputs [
    A {from=MaterialExpressionReroute_29}
    B {from=""}
   ]
   properties {constB=-1}
  }
  MaterialExpressionMultiply_108 {
   class=MaterialExpressionMultiply
   inputs [
    A {from=MaterialExpressionSaturate_39}
    B {from=MaterialExpressionNormalize_1}
   ]
  }
  MaterialExpressionMultiply_109 {
   class=MaterialExpressionMultiply
   inputs [
    A {from=MaterialExpressionConvert_18}
    B {from=MaterialExpressionNamedRerouteUsage_177}
   ]
  }
  MaterialExpressionMultiply_11 {
   class=MaterialExpressionMultiply
   inputs [
    A {from=MaterialExpressionMultiply_19}
    B {from=MaterialExpressionScalarParameter_22}
   ]
  }
  MaterialExpressionMultiply_110 {
   class=MaterialExpressionMultiply
   inputs [
    A {from=MaterialExpressionScalarParameter_26}
    B {from=""}
   ]
   properties {constB=-1}
  }
  MaterialExpressionMultiply_111 {
   class=MaterialExpressionMultiply
   inputs [
    A {from=MaterialExpressionConvert_6}
    B {from=MaterialExpressionStaticSwitchParameter_11}
   ]
  }
  MaterialExpressionMultiply_112 {
   class=MaterialExpressionMultiply
   inputs [
    A {from=MaterialExpressionMultiply_116}
    B {from=MaterialExpressionLinearInterpolate_49}
   ]
   properties {constB=0.3}
  }
  MaterialExpressionMultiply_113 {
   class=MaterialExpressionMultiply
   inputs [
    A {from=MaterialExpressionAdd_41}
    B {from=MaterialExpressionScalarParameter_47}
   ]
  }
  MaterialExpressionMultiply_114 {
   class=MaterialExpressionMultiply
   inputs [
    A {from=MaterialExpressionConvert_9}
    B {from=MaterialExpressionNamedRerouteUsage_176}
   ]
  }
  MaterialExpressionMultiply_115 {
   class=MaterialExpressionMultiply
   inputs [
    A {from=MaterialExpressionMaterialFunctionCall_13, output=B}
    B {from=""}
   ]
   properties {constB=5}
  }
  MaterialExpressionMultiply_116 {
   class=MaterialExpressionMultiply
   inputs [
    A {from=MaterialExpressionNamedRerouteUsage_99}
    B {from=MaterialExpressionLinearInterpolate_47}
   ]
  }
  MaterialExpressionMultiply_117 {
   class=MaterialExpressionMultiply
   inputs [
    A {from=MaterialExpressionReroute_33}
    B {from=MaterialExpressionNamedRerouteUsage_100}
   ]
  }
  MaterialExpressionMultiply_118 {
   class=MaterialExpressionMultiply
   inputs [
    A {from=MaterialExpressionSaturate_41}
    B {from=""}
   ]
   properties {constB=2}
  }
  MaterialExpressionMultiply_119 {
   class=MaterialExpressionMultiply
   inputs [
    A {from=MaterialExpressionMaterialFunctionCall_31, output=Result}
    B {from=MaterialExpressionMaterialFunctionCall_13, output=G}
   ]
  }
  MaterialExpressionMultiply_12 {
   class=MaterialExpressionMultiply
   inputs [
    A {from=MaterialExpressionScalarParameter_13}
    B {from=""}
   ]
   properties {constB=4}
  }
  MaterialExpressionMultiply_122 {
   class=MaterialExpressionMultiply
   inputs [
    A {from=MaterialExpressionAdd_46}
    B {from=MaterialExpressionNamedRerouteUsage_92}
   ]
  }
  MaterialExpressionMultiply_123 {
   class=MaterialExpressionMultiply
   inputs [
    A {from=MaterialExpressionNamedRerouteUsage_106}
    B {from=MaterialExpressionScalarParameter_50}
   ]
  }
  MaterialExpressionMultiply_125 {
   class=MaterialExpressionMultiply
   inputs [
    A {from=MaterialExpressionLinearInterpolate_46}
    B {from=MaterialExpressionScalarParameter_51}
   ]
  }
  MaterialExpressionMultiply_127 {
   class=MaterialExpressionMultiply
   inputs [
    A {from=MaterialExpressionMultiply_129}
    B {from=MaterialExpressionReroute_4}
   ]
  }
  MaterialExpressionMultiply_129 {
   class=MaterialExpressionMultiply
   inputs [
    A {from=MaterialExpressionSaturate_48}
    B {from=MaterialExpressionDesaturation_6}
   ]
  }
  MaterialExpressionMultiply_130 {
   class=MaterialExpressionMultiply
   inputs [
    A {from=MaterialExpressionLinearInterpolate_53}
    B {from=MaterialExpressionScalarParameter_82}
   ]
  }
  MaterialExpressionMultiply_133 {
   class=MaterialExpressionMultiply
   inputs [
    A {from=MaterialExpressionStaticSwitchParameter_10}
    B {from=MaterialExpressionMultiply_154}
   ]
  }
  MaterialExpressionMultiply_134 {
   class=MaterialExpressionMultiply
   inputs [
    A {from=MaterialExpressionReroute_36}
    B {from=MaterialExpressionLinearInterpolate_38}
   ]
  }
  MaterialExpressionMultiply_136 {
   class=MaterialExpressionMultiply
   inputs [
    A {from=MaterialExpressionStaticSwitchParameter_7}
    B {from=""}
   ]
   properties {constB=2}
  }
  MaterialExpressionMultiply_137 {
   class=MaterialExpressionMultiply
   inputs [
    A {from=MaterialExpressionStaticSwitchParameter_8}
    B {from=""}
   ]
   properties {constB=0.5}
  }
  MaterialExpressionMultiply_138 {
   class=MaterialExpressionMultiply
   inputs [
    A {from=MaterialExpressionNamedRerouteUsage_115}
    B {from=MaterialExpressionNamedRerouteUsage_115}
   ]
  }
  MaterialExpressionMultiply_139 {
   class=MaterialExpressionMultiply
   inputs [
    A {from=MaterialExpressionLinearInterpolate_60}
    B {from=MaterialExpressionLinearInterpolate_61}
   ]
  }
  MaterialExpressionMultiply_14 {
   class=MaterialExpressionMultiply
   inputs [
    A {from=MaterialExpressionNamedRerouteUsage_62}
    B {from=MaterialExpressionFmod_1}
   ]
  }
  MaterialExpressionMultiply_140 {
   class=MaterialExpressionMultiply
   inputs [
    A {from=MaterialExpressionOneMinus_13}
    B {from=MaterialExpressionScalarParameter_69}
   ]
   properties {constB=0.7}
  }
  MaterialExpressionMultiply_141 {
   class=MaterialExpressionMultiply
   inputs [
    A {from=MaterialExpressionConvert_10}
    B {from=MaterialExpressionNamedRerouteUsage_175}
   ]
  }
  MaterialExpressionMultiply_142 {
   class=MaterialExpressionMultiply
   inputs [
    A {from=MaterialExpressionMaterialFunctionCall_5, output=Result}
    B {from=""}
   ]
   properties {constB=3}
  }
  MaterialExpressionMultiply_144 {
   class=MaterialExpressionMultiply
   inputs [
    A {from=MaterialExpressionMultiply_146}
    B {from=MaterialExpressionSmoothStep_1}
   ]
  }
  MaterialExpressionMultiply_145 {
   class=MaterialExpressionMultiply
   inputs [
    A {from=MaterialExpressionMaterialFunctionCall_31, output=Result}
    B {from=MaterialExpressionMaterialFunctionCall_13, output=B}
   ]
  }
  MaterialExpressionMultiply_146 {
   class=MaterialExpressionMultiply
   inputs [
    A {from=MaterialExpressionAdd_52}
    B {from=MaterialExpressionConstant_45}
   ]
   properties {constB=8}
  }
  MaterialExpressionMultiply_147 {
   class=MaterialExpressionMultiply
   inputs [
    A {from=MaterialExpressionChannelMaskParameter_1, output=R}
    B {from=MaterialExpressionMax_5}
   ]
  }
  MaterialExpressionMultiply_148 {
   class=MaterialExpressionMultiply
   inputs [
    A {from=MaterialExpressionOneMinus_15}
    B {from=MaterialExpressionScalarParameter_76}
   ]
  }
  MaterialExpressionMultiply_149 {
   class=MaterialExpressionMultiply
   inputs [
    A {from=MaterialExpressionOneMinus_20}
    B {from=MaterialExpressionScalarParameter_77}
   ]
  }
  MaterialExpressionMultiply_15 {
   class=MaterialExpressionMultiply
   inputs [
    A {from=MaterialExpressionLinearInterpolate_22}
    B {from=""}
   ]
   properties {constB=0.8}
  }
  MaterialExpressionMultiply_150 {
   class=MaterialExpressionMultiply
   inputs [
    A {from=MaterialExpressionOneMinus_21}
    B {from=MaterialExpressionScalarParameter_78}
   ]
  }
  MaterialExpressionMultiply_151 {
   class=MaterialExpressionMultiply
   inputs [
    A {from=MaterialExpressionOneMinus_22}
    B {from=MaterialExpressionScalarParameter_79}
   ]
  }
  MaterialExpressionMultiply_152 {
   class=MaterialExpressionMultiply
   inputs [
    A {from=MaterialExpressionReroute_46}
    B {from=MaterialExpressionAdd_57}
   ]
  }
  MaterialExpressionMultiply_153 {
   class=MaterialExpressionMultiply
   inputs [
    A {from=MaterialExpressionGetMaterialAttributes_3, output=EmissiveColor}
    B {from=MaterialExpressionSaturate_63}
   ]
  }
  MaterialExpressionMultiply_154 {
   class=MaterialExpressionMultiply
   inputs [
    A {from=MaterialExpressionReroute_42}
    B {from=""}
   ]
   properties {constB=0.6}
  }
  MaterialExpressionMultiply_155 {
   class=MaterialExpressionMultiply
   inputs [
    A {from=MaterialExpressionNamedRerouteUsage_113}
    B {from=""}
   ]
   properties {constB=0.5}
  }
  MaterialExpressionMultiply_156 {
   class=MaterialExpressionMultiply
   inputs [
    A {from=MaterialExpressionMaterialFunctionCall_23, output=G}
    B {from=MaterialExpressionNamedRerouteUsage_155}
   ]
  }
  MaterialExpressionMultiply_157 {
   class=MaterialExpressionMultiply
   inputs [
    A {from=MaterialExpressionNamedRerouteUsage_158}
    B {from=MaterialExpressionMaterialFunctionCall_23, output=R}
   ]
  }
  MaterialExpressionMultiply_158 {
   class=MaterialExpressionMultiply
   inputs [
    A {from=MaterialExpressionMaterialFunctionCall_19, output=G}
    B {from=MaterialExpressionNamedRerouteUsage_161}
   ]
  }
  MaterialExpressionMultiply_159 {
   class=MaterialExpressionMultiply
   inputs [
    A {from=MaterialExpressionNamedRerouteUsage_164}
    B {from=MaterialExpressionMaterialFunctionCall_19, output=R}
   ]
  }
  MaterialExpressionMultiply_16 {
   class=MaterialExpressionMultiply
   inputs [
    A {from=MaterialExpressionNamedRerouteDeclaration_11}
    B {from=""}
   ]
   properties {constB=0.05}
  }
  MaterialExpressionMultiply_161 {
   class=MaterialExpressionMultiply
   inputs [
    A {from=MaterialExpressionNamedRerouteUsage_103}
    B {from=MaterialExpressionReroute_33}
   ]
  }
  MaterialExpressionMultiply_162 {
   class=MaterialExpressionMultiply
   inputs [
    A {from=MaterialExpressionStaticSwitchParameter_13}
    B {from=MaterialExpressionLinearInterpolate_68}
   ]
  }
  MaterialExpressionMultiply_163 {
   class=MaterialExpressionMultiply
   inputs [
    A {from=MaterialExpressionStaticComponentMaskParameter_0}
    B {from=MaterialExpressionScalarParameter_103}
   ]
   properties {constB=3}
  }
  MaterialExpressionMultiply_17 {
   class=MaterialExpressionMultiply
   inputs [
    A {from=MaterialExpressionStaticSwitchParameter_24}
    B {from=MaterialExpressionRayTracingQualitySwitch_0}
   ]
  }
  MaterialExpressionMultiply_173 {
   class=MaterialExpressionMultiply
   inputs [
    A {from=MaterialExpressionReroute_43}
    B {from=MaterialExpressionScalarParameter_88}
   ]
  }
  MaterialExpressionMultiply_174 {
   class=MaterialExpressionMultiply
   inputs [
    A {from=MaterialExpressionScalarParameter_92}
    B {from=""}
   ]
   properties {constB=3.5}
  }
  MaterialExpressionMultiply_175 {
   class=MaterialExpressionMultiply
   inputs [
    A {from=MaterialExpressionMaterialFunctionCall_39, output=G}
    B {from=MaterialExpressionNamedRerouteUsage_152}
   ]
  }
  MaterialExpressionMultiply_176 {
   class=MaterialExpressionMultiply
   inputs [
    A {from=MaterialExpressionNamedRerouteUsage_151}
    B {from=MaterialExpressionMaterialFunctionCall_39, output=R}
   ]
  }
  MaterialExpressionMultiply_177 {
   class=MaterialExpressionMultiply
   inputs [
    A {from=MaterialExpressionMaterialFunctionCall_40, output=G}
    B {from=MaterialExpressionNamedRerouteUsage_150}
   ]
  }
  MaterialExpressionMultiply_178 {
   class=MaterialExpressionMultiply
   inputs [
    A {from=MaterialExpressionNamedRerouteUsage_149}
    B {from=MaterialExpressionMaterialFunctionCall_40, output=R}
   ]
  }
  MaterialExpressionMultiply_179 {
   class=MaterialExpressionMultiply
   inputs [
    A {from=MaterialExpressionMaterialFunctionCall_41, output=G}
    B {from=MaterialExpressionNamedRerouteUsage_145}
   ]
  }
  MaterialExpressionMultiply_18 {
   class=MaterialExpressionMultiply
   inputs [
    A {from=MaterialExpressionSaturate_7}
    B {from=MaterialExpressionAdd_15}
   ]
  }
  MaterialExpressionMultiply_180 {
   class=MaterialExpressionMultiply
   inputs [
    A {from=MaterialExpressionNamedRerouteUsage_146}
    B {from=MaterialExpressionMaterialFunctionCall_41, output=R}
   ]
  }
  MaterialExpressionMultiply_183 {
   class=MaterialExpressionMultiply
   inputs [
    A {from=MaterialExpressionMaterialFunctionCall_43, output=G}
    B {from=MaterialExpressionNamedRerouteUsage_154}
   ]
  }
  MaterialExpressionMultiply_184 {
   class=MaterialExpressionMultiply
   inputs [
    A {from=MaterialExpressionNamedRerouteUsage_153}
    B {from=MaterialExpressionMaterialFunctionCall_43, output=R}
   ]
  }
  MaterialExpressionMultiply_186 {
   class=MaterialExpressionMultiply
   inputs [
    A {from=MaterialExpressionMultiply_199}
    B {from=MaterialExpressionReroute_32}
   ]
  }
  MaterialExpressionMultiply_187 {
   class=MaterialExpressionMultiply
   inputs [
    A {from=MaterialExpressionDesaturation_7}
    B {from=MaterialExpressionScalarParameter_101}
   ]
  }
  MaterialExpressionMultiply_19 {
   class=MaterialExpressionMultiply
   inputs [
    A {from=MaterialExpressionNormalize_0}
    B {from=""}
   ]
   properties {constB=-1}
  }
  MaterialExpressionMultiply_190 {
   class=MaterialExpressionMultiply
   inputs [
    A {from=MaterialExpressionConvert_6}
    B {from=""}
   ]
   properties {constB=10}
  }
  MaterialExpressionMultiply_196 {
   class=MaterialExpressionMultiply
   inputs [
    A {from=MaterialExpressionMultiply_119}
    B {from=""}
   ]
   properties {constB=3}
  }
  MaterialExpressionMultiply_198 {
   class=MaterialExpressionMultiply
   inputs [
    A {from=MaterialExpressionNamedRerouteUsage_172}
    B {from=MaterialExpressionScalarParameter_106}
   ]
   properties {constB=0.2}
  }
  MaterialExpressionMultiply_199 {
   class=MaterialExpressionMultiply
   inputs [
    A {from=MaterialExpressionStaticSwitchParameter_12}
    B {from=""}
   ]
   properties {constB=0.4}
  }
  MaterialExpressionMultiply_2 {
   class=MaterialExpressionMultiply
   inputs [
    A {from=MaterialExpressionNamedRerouteUsage_13}
    B {from=MaterialExpressionLinearInterpolate_43}
   ]
  }
  MaterialExpressionMultiply_20 {
   class=MaterialExpressionMultiply
   inputs [
    A {from=MaterialExpressionNamedRerouteUsage_33}
    B {from=MaterialExpressionNamedRerouteUsage_33}
   ]
  }
  MaterialExpressionMultiply_204 {
   class=MaterialExpressionMultiply
   inputs [
    A {from=MaterialExpressionMultiply_206}
    B {from=MaterialExpressionReroute_55}
   ]
  }
  MaterialExpressionMultiply_206 {
   class=MaterialExpressionMultiply
   inputs [
    A {from=MaterialExpressionNamedRerouteUsage_179}
    B {from=MaterialExpressionNamedRerouteUsage_180}
   ]
  }
  MaterialExpressionMultiply_207 {
   class=MaterialExpressionMultiply
   inputs [
    A {from=MaterialExpressionAdd_76}
    B {from=""}
   ]
   properties {constB=0.2}
  }
  MaterialExpressionMultiply_21 {
   class=MaterialExpressionMultiply
   inputs [
    A {from=MaterialExpressionFmod_0}
    B {from=MaterialExpressionMaterialFunctionCall_9, output=Result}
   ]
  }
  MaterialExpressionMultiply_24 {
   class=MaterialExpressionMultiply
   inputs [
    A {from=MaterialExpressionNamedRerouteUsage_18}
    B {from=""}
   ]
   properties {constB=1.2}
  }
  MaterialExpressionMultiply_26 {
   class=MaterialExpressionMultiply
   inputs [
    A {from=MaterialExpressionMultiply_32}
    B {from=MaterialExpressionReroute_6}
   ]
  }
  MaterialExpressionMultiply_27 {
   class=MaterialExpressionMultiply
   inputs [
    A {from=MaterialExpressionScalarParameter_92}
    B {from=""}
   ]
   properties {constB=2}
  }
  MaterialExpressionMultiply_29 {
   class=MaterialExpressionMultiply
   inputs [
    A {from=MaterialExpressionNormalize_2}
    B {from=""}
   ]
   properties {constB=-1}
  }
  MaterialExpressionMultiply_3 {
   class=MaterialExpressionMultiply
   inputs [
    A {from=MaterialExpressionStaticSwitchParameter_2}
    B {from=MaterialExpressionSaturate_1}
   ]
  }
  MaterialExpressionMultiply_31 {
   class=MaterialExpressionMultiply
   inputs [
    A {from=MaterialExpressionNamedRerouteUsage_25}
    B {from=""}
   ]
   properties {constB=0.7}
  }
  MaterialExpressionMultiply_32 {
   class=MaterialExpressionMultiply
   inputs [
    A {from=MaterialExpressionSaturate_3}
    B {from=MaterialExpressionDesaturation_1}
   ]
  }
  MaterialExpressionMultiply_33 {
   class=MaterialExpressionMultiply
   inputs [
    A {from=MaterialExpressionNamedRerouteUsage_63}
    B {from=""}
   ]
   properties {constB=0.8}
  }
  MaterialExpressionMultiply_35 {
   class=MaterialExpressionMultiply
   inputs [
    A {from=MaterialExpressionMultiply_173}
    B {from=MaterialExpressionNamedRerouteUsage_59}
   ]
  }
  MaterialExpressionMultiply_37 {
   class=MaterialExpressionMultiply
   inputs [
    A {from=MaterialExpressionNamedRerouteUsage_139}
    B {from=MaterialExpressionScalarParameter_2}
   ]
  }
  MaterialExpressionMultiply_4 {
   class=MaterialExpressionMultiply
   inputs [
    A {from=MaterialExpressionNamedRerouteDeclaration_83}
    B {from=MaterialExpressionLinearInterpolate_2}
   ]
  }
  MaterialExpressionMultiply_42 {
   class=MaterialExpressionMultiply
   inputs [
    A {from=MaterialExpressionMultiply_24}
    B {from=MaterialExpressionMultiply_24}
   ]
  }
  MaterialExpressionMultiply_5 {
   class=MaterialExpressionMultiply
   inputs [
    A {from=MaterialExpressionLinearInterpolate_11}
    B {from=""}
   ]
   properties {constB=0.6}
  }
  MaterialExpressionMultiply_51 {
   class=MaterialExpressionMultiply
   inputs [
    A {from=MaterialExpressionSine_2}
    B {from=""}
   ]
   properties {constB=0.5}
  }
  MaterialExpressionMultiply_52 {
   class=MaterialExpressionMultiply
   inputs [
    A {from=MaterialExpressionOneMinus_6}
    B {from=MaterialExpressionMultiply_134}
   ]
  }
  MaterialExpressionMultiply_53 {
   class=MaterialExpressionMultiply
   inputs [
    A {from=MaterialExpressionMultiply_101}
    B {from=MaterialExpressionConstant_28}
   ]
   properties {constB=8}
  }
  MaterialExpressionMultiply_6 {
   class=MaterialExpressionMultiply
   inputs [
    A {from=MaterialExpressionVectorParameter_1, output=RGB}
    B {from=MaterialExpressionScalarParameter_10}
   ]
  }
  MaterialExpressionMultiply_68 {
   class=MaterialExpressionMultiply
   inputs [
    A {from=MaterialExpressionMultiply_1}
    B {from=MaterialExpressionScalarParameter_1}
   ]
  }
  MaterialExpressionMultiply_7 {
   class=MaterialExpressionMultiply
   inputs [
    A {from=MaterialExpressionVectorParameter_3, output=RGB}
    B {from=""}
   ]
   properties {constB=2}
  }
  MaterialExpressionMultiply_8 {
   class=MaterialExpressionMultiply
   inputs [
    A {from=MaterialExpressionStaticSwitchParameter_20}
    B {from=MaterialExpressionNamedRerouteUsage_23}
   ]
  }
  MaterialExpressionMultiply_9 {
   class=MaterialExpressionMultiply
   inputs [
    A {from=MaterialExpressionMultiply_103}
    B {from=MaterialExpressionNamedRerouteUsage_24}
   ]
  }
  MaterialExpressionMultiply_98 {
   class=MaterialExpressionMultiply
   inputs [
    A {from=MaterialExpressionReroute_28}
    B {from=MaterialExpressionScalarParameter_8}
   ]
   properties {constB=0.3}
  }
  MaterialExpressionNamedRerouteDeclaration_0 {
   class=MaterialExpressionNamedRerouteDeclaration
   inputs [None {from=MaterialExpressionSaturate_2}]
   properties {name=NightTime, nodeColor=<rgba 0 1 0.6016 1>}
  }
  MaterialExpressionNamedRerouteDeclaration_1 {
   class=MaterialExpressionNamedRerouteDeclaration
   inputs [None {from=MaterialExpressionMaterialFunctionCall_14, output=Random time}]
   properties {name=Random Time, nodeColor=<rgba 0 1 0.6484 1>}
  }
  MaterialExpressionNamedRerouteDeclaration_10 {
   class=MaterialExpressionNamedRerouteDeclaration
   inputs [None {from=MaterialExpressionOneMinus_7}]
   properties {name=Glow, nodeColor=<rgba 0.6641 0 1 1>}
  }
  MaterialExpressionNamedRerouteDeclaration_11 {
   class=MaterialExpressionNamedRerouteDeclaration
   inputs [None {from=MaterialExpressionSaturate_6}]
   properties {name=Square Color glitch, nodeColor=<rgba 0.2422 1 0 1>}
  }
  MaterialExpressionNamedRerouteDeclaration_12 {
   class=MaterialExpressionNamedRerouteDeclaration
   inputs [None {from=MaterialExpressionReroute_0}]
   properties {name=Dots, nodeColor=<rgba 0 1 0.2266 1>}
  }
  MaterialExpressionNamedRerouteDeclaration_13 {
   class=MaterialExpressionNamedRerouteDeclaration
   inputs [None {from=MaterialExpressionVectorParameter_13, output=RGB}]
   properties {name=Logo Color 2, nodeColor=<rgba 1 0.9921 0.914 1>}
  }
  MaterialExpressionNamedRerouteDeclaration_14 {
   class=MaterialExpressionNamedRerouteDeclaration
   inputs [None {from=MaterialExpressionStaticSwitchParameter_32}]
   properties {name=Logo Color, nodeColor=<rgba 0.0578 0.4793 1 1>}
  }
  MaterialExpressionNamedRerouteDeclaration_15 {
   class=MaterialExpressionNamedRerouteDeclaration
   inputs [None {from=MaterialExpressionVectorParameter_16, output=RGB}]
   properties {name=Glass Color 2, nodeColor=<rgba 0 0.6499 0.9219 1>}
  }
  MaterialExpressionNamedRerouteDeclaration_16 {
   class=MaterialExpressionNamedRerouteDeclaration
   inputs [None {from=MaterialExpressionVectorParameter_14, output=RGB}]
   properties {name=Glass Color 1, nodeColor=<rgba 0 0.1779 0.2705 1>}
  }
  MaterialExpressionNamedRerouteDeclaration_17 {
   class=MaterialExpressionNamedRerouteDeclaration
   inputs [None {from=MaterialExpressionLinearInterpolate_42}]
   properties {name=Dot Emissive, nodeColor=<rgba 1 0 0.7969 1>}
  }
  MaterialExpressionNamedRerouteDeclaration_18 {
   class=MaterialExpressionNamedRerouteDeclaration
   inputs [None {from=MaterialExpressionMaterialFunctionCall_14, output=GlitchFrequencyMask}]
   properties {name=GlitchFrequencyMask, nodeColor=<rgba 0.9377 1 0.8953 1>}
  }
  MaterialExpressionNamedRerouteDeclaration_2 {
   class=MaterialExpressionNamedRerouteDeclaration
   inputs [None {from=MaterialExpressionStaticSwitchParameter_4}]
   properties {name=DFX BG, nodeColor=<rgba 0 0.7656 1 1>}
  }
  MaterialExpressionNamedRerouteDeclaration_20 {
   class=MaterialExpressionNamedRerouteDeclaration
   inputs [None {from=MaterialExpressionLinearInterpolate_29}]
   properties {name=Logo Color Mask, nodeColor=<rgba 0.0612 0.0021 0.0203 1>}
  }
  MaterialExpressionNamedRerouteDeclaration_22 {
   class=MaterialExpressionNamedRerouteDeclaration
   inputs [None {from=MaterialExpressionMultiply_14}]
   properties {name=Color Flash, nodeColor=<rgba 0.5234 1 0 1>}
  }
  MaterialExpressionNamedRerouteDeclaration_29 {
   class=MaterialExpressionNamedRerouteDeclaration
   inputs [None {from=MaterialExpressionMultiply_9}]
   properties {name=Logo Shadow, nodeColor=<rgba 0 0.1328 1 1>}
  }
  MaterialExpressionNamedRerouteDeclaration_3 {
   class=MaterialExpressionNamedRerouteDeclaration
   inputs [None {from=MaterialExpressionTextureSampleParameter2D_2, output=RGBA}]
   properties {name=DFX, nodeColor=<rgba 0 0.7656 1 1>}
  }
  MaterialExpressionNamedRerouteDeclaration_35 {
   class=MaterialExpressionNamedRerouteDeclaration
   inputs [None {from=MaterialExpressionMultiply_196}]
   properties {name=Emitting Lights, nodeColor=<rgba 0 0.4844 1 1>}
  }
  MaterialExpressionNamedRerouteDeclaration_4 {
   class=MaterialExpressionNamedRerouteDeclaration
   inputs [None {from=MaterialExpressionSaturate_68}]
   properties {name=Depth Layer 2, nodeColor=<rgba 1 0 0.0469 1>}
  }
  MaterialExpressionNamedRerouteDeclaration_44 {
   class=MaterialExpressionNamedRerouteDeclaration
   inputs [None {from=MaterialExpressionDeriveNormalZ_1}]
   properties {name=Circle Normal, nodeColor=<rgba 1 0.3281 0 1>}
  }
  MaterialExpressionNamedRerouteDeclaration_45 {
   class=MaterialExpressionNamedRerouteDeclaration
   inputs [None {from=MaterialExpressionLinearInterpolate_63}]
   properties {name=UV Layer 3, nodeColor=<rgba 0 1 0.4844 1>}
  }
  MaterialExpressionNamedRerouteDeclaration_46 {
   class=MaterialExpressionNamedRerouteDeclaration
   inputs [None {from=MaterialExpressionSaturate_40}]
   properties {name=Distance Lerp, nodeColor=<rgba 1 0.008 0.0622 1>}
  }
  MaterialExpressionNamedRerouteDeclaration_47 {
   class=MaterialExpressionNamedRerouteDeclaration
   inputs [None {from=MaterialExpressionMaterialFunctionCall_20, output=Result}]
   properties {name=Max Depth L4, nodeColor=<rgba 1 0.0697 0.092 1>}
  }
  MaterialExpressionNamedRerouteDeclaration_48 {
   class=MaterialExpressionNamedRerouteDeclaration
   inputs [None {from=MaterialExpressionSaturate_43}]
   properties {name=Pixel Size, nodeColor=<rgba 0 0.1094 1 1>}
  }
  MaterialExpressionNamedRerouteDeclaration_49 {
   class=MaterialExpressionNamedRerouteDeclaration
   inputs [None {from=MaterialExpressionLinearInterpolate_76}]
   properties {name=Dot Mask Layer 3, nodeColor=<rgba 0.8017 0 1 1>}
  }
  MaterialExpressionNamedRerouteDeclaration_5 {
   class=MaterialExpressionNamedRerouteDeclaration
   inputs [None {from=MaterialExpressionMultiply_16}]
   properties {name=Glitch Noise, nodeColor=<rgba 0 1 0.6953 1>}
  }
  MaterialExpressionNamedRerouteDeclaration_50 {
   class=MaterialExpressionNamedRerouteDeclaration
   inputs [None {from=MaterialExpressionMultiply_8}]
   properties {name=Logo, nodeColor=<rgba 1 0 0.6797 1>}
  }
  MaterialExpressionNamedRerouteDeclaration_51 {
   class=MaterialExpressionNamedRerouteDeclaration
   inputs [None {from=MaterialExpressionStaticSwitchParameter_29}]
   properties {name=Logo Mask, nodeColor=<rgba 0.125 1 0 1>}
  }
  MaterialExpressionNamedRerouteDeclaration_52 {
   class=MaterialExpressionNamedRerouteDeclaration
   inputs [None {from=MaterialExpressionConvert_4}]
   properties {name=UV Layer 2, nodeColor=<rgba 0.8047 0 1 1>}
  }
  MaterialExpressionNamedRerouteDeclaration_53 {
   class=MaterialExpressionNamedRerouteDeclaration
   inputs [None {from=MaterialExpressionStaticSwitchParameter_22}]
   properties {name=Base color, nodeColor=<rgba 0 0.1797 1 1>}
  }
  MaterialExpressionNamedRerouteDeclaration_54 {
   class=MaterialExpressionNamedRerouteDeclaration
   inputs [None {from=MaterialExpressionStaticSwitchParameter_21}]
   properties {name=Base color_2, nodeColor=<rgba 0 0.1797 1 1>}
  }
  MaterialExpressionNamedRerouteDeclaration_56 {
   class=MaterialExpressionNamedRerouteDeclaration
   inputs [None {from=MaterialExpressionVertexInterpolator_3, output=PS}]
   properties {name=Inverted Fresnel, nodeColor=<rgba 0 1 0.4609 1>}
  }
  MaterialExpressionNamedRerouteDeclaration_57 {
   class=MaterialExpressionNamedRerouteDeclaration
   inputs [None {from=MaterialExpressionSaturate_54}]
   properties {name=Dot Mask Layer 4, nodeColor=<rgba 0.8017 0 1 1>}
  }
  MaterialExpressionNamedRerouteDeclaration_58 {
   class=MaterialExpressionNamedRerouteDeclaration
   inputs [None {from=MaterialExpressionTextureSample_7, output=B}]
   properties {name=Layer 4 gradient, nodeColor=<rgba 0.0078 1 0 1>}
  }
  MaterialExpressionNamedRerouteDeclaration_59 {
   class=MaterialExpressionNamedRerouteDeclaration
   inputs [None {from=MaterialExpressionFunctionInput_3}]
   properties {name=Original MA, nodeColor=<rgba 0.7109 0 1 1>}
  }
  MaterialExpressionNamedRerouteDeclaration_6 {
   class=MaterialExpressionNamedRerouteDeclaration
   inputs [None {from=MaterialExpressionStaticSwitchParameter_3}]
   properties {name=Normal, nodeColor=<rgba 0.1484 0 1 1>}
  }
  MaterialExpressionNamedRerouteDeclaration_60 {
   class=MaterialExpressionNamedRerouteDeclaration
   inputs [None {from=MaterialExpressionStaticSwitchParameter_19}]
   properties {name=Emissive color, nodeColor=<rgba 1 0 0.7969 1>}
  }
  MaterialExpressionNamedRerouteDeclaration_61 {
   class=MaterialExpressionNamedRerouteDeclaration
   inputs [None {from=MaterialExpressionMaterialFunctionCall_34, output=Result}]
   properties {name=OG Texture Mask, nodeColor=<rgba 0 1 0.5547 1>}
  }
  MaterialExpressionNamedRerouteDeclaration_62 {
   class=MaterialExpressionNamedRerouteDeclaration
   inputs [None {from=MaterialExpressionMaterialFunctionCall_35, output=DDX}]
   properties {name=DDX, nodeColor=<rgba 1 0 0.2578 1>}
  }
  MaterialExpressionNamedRerouteDeclaration_63 {
   class=MaterialExpressionNamedRerouteDeclaration
   inputs [None {from=MaterialExpressionMaterialFunctionCall_35, output=DDY}]
   properties {name=DDY, nodeColor=<rgba 1 0 0.2578 1>}
  }
  MaterialExpressionNamedRerouteDeclaration_64 {
   class=MaterialExpressionNamedRerouteDeclaration
   inputs [None {from=MaterialExpressionMaterialFunctionCall_35, output=UVs}]
   properties {name=Triplanar UV, nodeColor=<rgba 1 0 0.2578 1>}
  }
  MaterialExpressionNamedRerouteDeclaration_66 {
   class=MaterialExpressionNamedRerouteDeclaration
   inputs [None {from=MaterialExpressionLinearInterpolate_74}]
   properties {name=UV Layer 3 DFX, nodeColor=<rgba 0 1 0.4844 1>}
  }
  MaterialExpressionNamedRerouteDeclaration_67 {
   class=MaterialExpressionNamedRerouteDeclaration
   inputs [None {from=MaterialExpressionStaticSwitchParameter_6}]
   properties {name=Max Depth DFX, nodeColor=<rgba 1 0.0697 0.092 1>}
  }
  MaterialExpressionNamedRerouteDeclaration_68 {
   class=MaterialExpressionNamedRerouteDeclaration
   inputs [None {from=MaterialExpressionAppendVector_4}]
   properties {name=UV DFX, nodeColor=<rgba 1 0 0.3516 1>}
  }
  MaterialExpressionNamedRerouteDeclaration_69 {
   class=MaterialExpressionNamedRerouteDeclaration
   inputs [None {from=MaterialExpressionSaturate_70}]
   properties {name=Distance Lerp Far, nodeColor=<rgba 1 0.008 0.0622 1>}
  }
  MaterialExpressionNamedRerouteDeclaration_7 {
   class=MaterialExpressionNamedRerouteDeclaration
   inputs [None {from=MaterialExpressionReroute_4}]
   properties {name=Color 2 mask, nodeColor=<rgba 0 1 0.2031 1>}
  }
  MaterialExpressionNamedRerouteDeclaration_70 {
   class=MaterialExpressionNamedRerouteDeclaration
   inputs [None {from=MaterialExpressionAdd_63}]
   properties {name="Dots  X Size", nodeColor=<rgba 0.8984 1 0 1>}
  }
  MaterialExpressionNamedRerouteDeclaration_71 {
   class=MaterialExpressionNamedRerouteDeclaration
   inputs [None {from=MaterialExpressionAdd_64}]
   properties {name=Dots Y Size, nodeColor=<rgba 0.8984 1 0 1>}
  }
  MaterialExpressionNamedRerouteDeclaration_72 {
   class=MaterialExpressionNamedRerouteDeclaration
   inputs [None {from=MaterialExpressionScalarParameter_99}]
   properties {name=Logo X Pos, nodeColor=<rgba 0 0.7048 1 1>}
  }
  MaterialExpressionNamedRerouteDeclaration_73 {
   class=MaterialExpressionNamedRerouteDeclaration
   inputs [None {from=MaterialExpressionScalarParameter_97}]
   properties {name=Logo Y Pos, nodeColor=<rgba 0 0.7048 1 1>}
  }
  MaterialExpressionNamedRerouteDeclaration_74 {
   class=MaterialExpressionNamedRerouteDeclaration
   inputs [None {from=MaterialExpressionAdd_65}]
   properties {name=Logo X Size, nodeColor=<rgba 0 0.7048 1 1>}
  }
  MaterialExpressionNamedRerouteDeclaration_75 {
   class=MaterialExpressionNamedRerouteDeclaration
   inputs [None {from=MaterialExpressionAdd_66}]
   properties {name=Logo Y Size, nodeColor=<rgba 0 0.7048 1 1>}
  }
  MaterialExpressionNamedRerouteDeclaration_76 {
   class=MaterialExpressionNamedRerouteDeclaration
   inputs [None {from=MaterialExpressionMultiply_116}]
   properties {name=Shadow Color, nodeColor=<rgba 0 0.0156 1 1>}
  }
  MaterialExpressionNamedRerouteDeclaration_79 {
   class=MaterialExpressionNamedRerouteDeclaration
   inputs [None {from=MaterialExpressionMaterialFunctionCall_13, output=B}]
   properties {name=TX Dot L4, nodeColor=<rgba 0 1 0.4375 1>}
  }
  MaterialExpressionNamedRerouteDeclaration_8 {
   class=MaterialExpressionNamedRerouteDeclaration
   inputs [None {from=MaterialExpressionOneMinus_4}]
   properties {name=NoiseAlpha, nodeColor=<rgba 0.8047 0 1 1>}
  }
  MaterialExpressionNamedRerouteDeclaration_80 {
   class=MaterialExpressionNamedRerouteDeclaration
   inputs [None {from=MaterialExpressionLinearInterpolate_47}]
   properties {name=Logo BG, nodeColor=<rgba 0 1 0.6016 1>}
  }
  MaterialExpressionNamedRerouteDeclaration_81 {
   class=MaterialExpressionNamedRerouteDeclaration
   inputs [None {from=MaterialExpressionScalarParameter_5}]
   properties {name=Dot Amount, nodeColor=<rgba 0.9453 0 1 1>}
  }
  MaterialExpressionNamedRerouteDeclaration_82 {
   class=MaterialExpressionNamedRerouteDeclaration
   inputs [None {from=MaterialExpressionMakeMaterialAttributes_1}]
   properties {name=Projector MA, nodeColor=<rgba 0.9688 1 0 1>}
  }
  MaterialExpressionNamedRerouteDeclaration_83 {
   class=MaterialExpressionNamedRerouteDeclaration
   inputs [None {from=MaterialExpressionTextureSample_7, output=R}]
   properties {name=Circuit, nodeColor=<rgba 0 1 0.0391 1>}
  }
  MaterialExpressionNamedRerouteDeclaration_9 {
   class=MaterialExpressionNamedRerouteDeclaration
   inputs [None {from=MaterialExpressionLinearInterpolate_28}]
   properties {name=UV Glitching, nodeColor=<rgba 1 0 0.3281 1>}
  }
  MaterialExpressionNamedRerouteUsage_0 {class=MaterialExpressionNamedRerouteUsage}
  MaterialExpressionNamedRerouteUsage_1 {class=MaterialExpressionNamedRerouteUsage}
  MaterialExpressionNamedRerouteUsage_10 {class=MaterialExpressionNamedRerouteUsage}
  MaterialExpressionNamedRerouteUsage_100 {class=MaterialExpressionNamedRerouteUsage}
  MaterialExpressionNamedRerouteUsage_102 {class=MaterialExpressionNamedRerouteUsage}
  MaterialExpressionNamedRerouteUsage_103 {class=MaterialExpressionNamedRerouteUsage}
  MaterialExpressionNamedRerouteUsage_104 {class=MaterialExpressionNamedRerouteUsage}
  MaterialExpressionNamedRerouteUsage_105 {class=MaterialExpressionNamedRerouteUsage}
  MaterialExpressionNamedRerouteUsage_106 {class=MaterialExpressionNamedRerouteUsage}
  MaterialExpressionNamedRerouteUsage_107 {class=MaterialExpressionNamedRerouteUsage}
  MaterialExpressionNamedRerouteUsage_108 {class=MaterialExpressionNamedRerouteUsage}
  MaterialExpressionNamedRerouteUsage_109 {class=MaterialExpressionNamedRerouteUsage}
  MaterialExpressionNamedRerouteUsage_11 {class=MaterialExpressionNamedRerouteUsage}
  MaterialExpressionNamedRerouteUsage_112 {class=MaterialExpressionNamedRerouteUsage}
  MaterialExpressionNamedRerouteUsage_113 {class=MaterialExpressionNamedRerouteUsage}
  MaterialExpressionNamedRerouteUsage_114 {class=MaterialExpressionNamedRerouteUsage}
  MaterialExpressionNamedRerouteUsage_115 {class=MaterialExpressionNamedRerouteUsage}
  MaterialExpressionNamedRerouteUsage_116 {class=MaterialExpressionNamedRerouteUsage}
  MaterialExpressionNamedRerouteUsage_119 {class=MaterialExpressionNamedRerouteUsage}
  MaterialExpressionNamedRerouteUsage_12 {class=MaterialExpressionNamedRerouteUsage}
  MaterialExpressionNamedRerouteUsage_120 {class=MaterialExpressionNamedRerouteUsage}
  MaterialExpressionNamedRerouteUsage_121 {class=MaterialExpressionNamedRerouteUsage}
  MaterialExpressionNamedRerouteUsage_122 {class=MaterialExpressionNamedRerouteUsage}
  MaterialExpressionNamedRerouteUsage_123 {class=MaterialExpressionNamedRerouteUsage}
  MaterialExpressionNamedRerouteUsage_124 {class=MaterialExpressionNamedRerouteUsage}
  MaterialExpressionNamedRerouteUsage_125 {class=MaterialExpressionNamedRerouteUsage}
  MaterialExpressionNamedRerouteUsage_126 {class=MaterialExpressionNamedRerouteUsage}
  MaterialExpressionNamedRerouteUsage_127 {class=MaterialExpressionNamedRerouteUsage}
  MaterialExpressionNamedRerouteUsage_128 {class=MaterialExpressionNamedRerouteUsage}
  MaterialExpressionNamedRerouteUsage_129 {class=MaterialExpressionNamedRerouteUsage}
  MaterialExpressionNamedRerouteUsage_13 {class=MaterialExpressionNamedRerouteUsage}
  MaterialExpressionNamedRerouteUsage_130 {class=MaterialExpressionNamedRerouteUsage}
  MaterialExpressionNamedRerouteUsage_131 {class=MaterialExpressionNamedRerouteUsage}
  MaterialExpressionNamedRerouteUsage_132 {class=MaterialExpressionNamedRerouteUsage}
  MaterialExpressionNamedRerouteUsage_133 {class=MaterialExpressionNamedRerouteUsage}
  MaterialExpressionNamedRerouteUsage_137 {class=MaterialExpressionNamedRerouteUsage}
  MaterialExpressionNamedRerouteUsage_138 {class=MaterialExpressionNamedRerouteUsage}
  MaterialExpressionNamedRerouteUsage_139 {class=MaterialExpressionNamedRerouteUsage}
  MaterialExpressionNamedRerouteUsage_14 {class=MaterialExpressionNamedRerouteUsage}
  MaterialExpressionNamedRerouteUsage_140 {class=MaterialExpressionNamedRerouteUsage}
  MaterialExpressionNamedRerouteUsage_141 {class=MaterialExpressionNamedRerouteUsage}
  MaterialExpressionNamedRerouteUsage_142 {class=MaterialExpressionNamedRerouteUsage}
  MaterialExpressionNamedRerouteUsage_143 {class=MaterialExpressionNamedRerouteUsage}
  MaterialExpressionNamedRerouteUsage_144 {class=MaterialExpressionNamedRerouteUsage}
  MaterialExpressionNamedRerouteUsage_145 {class=MaterialExpressionNamedRerouteUsage}
  MaterialExpressionNamedRerouteUsage_146 {class=MaterialExpressionNamedRerouteUsage}
  MaterialExpressionNamedRerouteUsage_149 {class=MaterialExpressionNamedRerouteUsage}
  MaterialExpressionNamedRerouteUsage_15 {class=MaterialExpressionNamedRerouteUsage}
  MaterialExpressionNamedRerouteUsage_150 {class=MaterialExpressionNamedRerouteUsage}
  MaterialExpressionNamedRerouteUsage_151 {class=MaterialExpressionNamedRerouteUsage}
  MaterialExpressionNamedRerouteUsage_152 {class=MaterialExpressionNamedRerouteUsage}
  MaterialExpressionNamedRerouteUsage_153 {class=MaterialExpressionNamedRerouteUsage}
  MaterialExpressionNamedRerouteUsage_154 {class=MaterialExpressionNamedRerouteUsage}
  MaterialExpressionNamedRerouteUsage_155 {class=MaterialExpressionNamedRerouteUsage}
  MaterialExpressionNamedRerouteUsage_156 {class=MaterialExpressionNamedRerouteUsage}
  MaterialExpressionNamedRerouteUsage_157 {class=MaterialExpressionNamedRerouteUsage}
  MaterialExpressionNamedRerouteUsage_158 {class=MaterialExpressionNamedRerouteUsage}
  MaterialExpressionNamedRerouteUsage_16 {class=MaterialExpressionNamedRerouteUsage}
  MaterialExpressionNamedRerouteUsage_161 {class=MaterialExpressionNamedRerouteUsage}
  MaterialExpressionNamedRerouteUsage_162 {class=MaterialExpressionNamedRerouteUsage}
  MaterialExpressionNamedRerouteUsage_163 {class=MaterialExpressionNamedRerouteUsage}
  MaterialExpressionNamedRerouteUsage_164 {class=MaterialExpressionNamedRerouteUsage}
  MaterialExpressionNamedRerouteUsage_167 {class=MaterialExpressionNamedRerouteUsage}
  MaterialExpressionNamedRerouteUsage_169 {class=MaterialExpressionNamedRerouteUsage}
  MaterialExpressionNamedRerouteUsage_17 {class=MaterialExpressionNamedRerouteUsage}
  MaterialExpressionNamedRerouteUsage_170 {class=MaterialExpressionNamedRerouteUsage}
  MaterialExpressionNamedRerouteUsage_172 {class=MaterialExpressionNamedRerouteUsage}
  MaterialExpressionNamedRerouteUsage_174 {class=MaterialExpressionNamedRerouteUsage}
  MaterialExpressionNamedRerouteUsage_175 {class=MaterialExpressionNamedRerouteUsage}
  MaterialExpressionNamedRerouteUsage_176 {class=MaterialExpressionNamedRerouteUsage}
  MaterialExpressionNamedRerouteUsage_177 {class=MaterialExpressionNamedRerouteUsage}
  MaterialExpressionNamedRerouteUsage_179 {class=MaterialExpressionNamedRerouteUsage}
  MaterialExpressionNamedRerouteUsage_18 {class=MaterialExpressionNamedRerouteUsage}
  MaterialExpressionNamedRerouteUsage_180 {class=MaterialExpressionNamedRerouteUsage}
  MaterialExpressionNamedRerouteUsage_181 {class=MaterialExpressionNamedRerouteUsage}
  MaterialExpressionNamedRerouteUsage_19 {class=MaterialExpressionNamedRerouteUsage}
  MaterialExpressionNamedRerouteUsage_2 {class=MaterialExpressionNamedRerouteUsage}
  MaterialExpressionNamedRerouteUsage_20 {class=MaterialExpressionNamedRerouteUsage}
  MaterialExpressionNamedRerouteUsage_21 {class=MaterialExpressionNamedRerouteUsage}
  MaterialExpressionNamedRerouteUsage_22 {class=MaterialExpressionNamedRerouteUsage}
  MaterialExpressionNamedRerouteUsage_23 {class=MaterialExpressionNamedRerouteUsage}
  MaterialExpressionNamedRerouteUsage_24 {class=MaterialExpressionNamedRerouteUsage}
  MaterialExpressionNamedRerouteUsage_25 {class=MaterialExpressionNamedRerouteUsage}
  MaterialExpressionNamedRerouteUsage_26 {class=MaterialExpressionNamedRerouteUsage}
  MaterialExpressionNamedRerouteUsage_27 {class=MaterialExpressionNamedRerouteUsage}
  MaterialExpressionNamedRerouteUsage_29 {class=MaterialExpressionNamedRerouteUsage}
  MaterialExpressionNamedRerouteUsage_3 {class=MaterialExpressionNamedRerouteUsage}
  MaterialExpressionNamedRerouteUsage_30 {class=MaterialExpressionNamedRerouteUsage}
  MaterialExpressionNamedRerouteUsage_31 {class=MaterialExpressionNamedRerouteUsage}
  MaterialExpressionNamedRerouteUsage_32 {class=MaterialExpressionNamedRerouteUsage}
  MaterialExpressionNamedRerouteUsage_33 {class=MaterialExpressionNamedRerouteUsage}
  MaterialExpressionNamedRerouteUsage_34 {class=MaterialExpressionNamedRerouteUsage}
  MaterialExpressionNamedRerouteUsage_35 {class=MaterialExpressionNamedRerouteUsage}
  MaterialExpressionNamedRerouteUsage_36 {class=MaterialExpressionNamedRerouteUsage}
  MaterialExpressionNamedRerouteUsage_37 {class=MaterialExpressionNamedRerouteUsage}
  MaterialExpressionNamedRerouteUsage_38 {class=MaterialExpressionNamedRerouteUsage}
  MaterialExpressionNamedRerouteUsage_39 {class=MaterialExpressionNamedRerouteUsage}
  MaterialExpressionNamedRerouteUsage_4 {class=MaterialExpressionNamedRerouteUsage}
  MaterialExpressionNamedRerouteUsage_40 {class=MaterialExpressionNamedRerouteUsage}
  MaterialExpressionNamedRerouteUsage_41 {class=MaterialExpressionNamedRerouteUsage}
  MaterialExpressionNamedRerouteUsage_42 {class=MaterialExpressionNamedRerouteUsage}
  MaterialExpressionNamedRerouteUsage_43 {class=MaterialExpressionNamedRerouteUsage}
  MaterialExpressionNamedRerouteUsage_44 {class=MaterialExpressionNamedRerouteUsage}
  MaterialExpressionNamedRerouteUsage_45 {class=MaterialExpressionNamedRerouteUsage}
  MaterialExpressionNamedRerouteUsage_46 {class=MaterialExpressionNamedRerouteUsage}
  MaterialExpressionNamedRerouteUsage_47 {class=MaterialExpressionNamedRerouteUsage}
  MaterialExpressionNamedRerouteUsage_48 {class=MaterialExpressionNamedRerouteUsage}
  MaterialExpressionNamedRerouteUsage_49 {class=MaterialExpressionNamedRerouteUsage}
  MaterialExpressionNamedRerouteUsage_5 {class=MaterialExpressionNamedRerouteUsage}
  MaterialExpressionNamedRerouteUsage_53 {class=MaterialExpressionNamedRerouteUsage}
  MaterialExpressionNamedRerouteUsage_54 {class=MaterialExpressionNamedRerouteUsage}
  MaterialExpressionNamedRerouteUsage_55 {class=MaterialExpressionNamedRerouteUsage}
  MaterialExpressionNamedRerouteUsage_56 {class=MaterialExpressionNamedRerouteUsage}
  MaterialExpressionNamedRerouteUsage_59 {class=MaterialExpressionNamedRerouteUsage}
  MaterialExpressionNamedRerouteUsage_6 {class=MaterialExpressionNamedRerouteUsage}
  MaterialExpressionNamedRerouteUsage_60 {class=MaterialExpressionNamedRerouteUsage}
  MaterialExpressionNamedRerouteUsage_61 {class=MaterialExpressionNamedRerouteUsage}
  MaterialExpressionNamedRerouteUsage_62 {class=MaterialExpressionNamedRerouteUsage}
  MaterialExpressionNamedRerouteUsage_63 {class=MaterialExpressionNamedRerouteUsage}
  MaterialExpressionNamedRerouteUsage_64 {class=MaterialExpressionNamedRerouteUsage}
  MaterialExpressionNamedRerouteUsage_65 {class=MaterialExpressionNamedRerouteUsage}
  MaterialExpressionNamedRerouteUsage_66 {class=MaterialExpressionNamedRerouteUsage}
  MaterialExpressionNamedRerouteUsage_67 {class=MaterialExpressionNamedRerouteUsage}
  MaterialExpressionNamedRerouteUsage_68 {class=MaterialExpressionNamedRerouteUsage}
  MaterialExpressionNamedRerouteUsage_69 {class=MaterialExpressionNamedRerouteUsage}
  MaterialExpressionNamedRerouteUsage_7 {class=MaterialExpressionNamedRerouteUsage}
  MaterialExpressionNamedRerouteUsage_70 {class=MaterialExpressionNamedRerouteUsage}
  MaterialExpressionNamedRerouteUsage_71 {class=MaterialExpressionNamedRerouteUsage}
  MaterialExpressionNamedRerouteUsage_72 {class=MaterialExpressionNamedRerouteUsage}
  MaterialExpressionNamedRerouteUsage_74 {class=MaterialExpressionNamedRerouteUsage}
  MaterialExpressionNamedRerouteUsage_8 {class=MaterialExpressionNamedRerouteUsage}
  MaterialExpressionNamedRerouteUsage_89 {class=MaterialExpressionNamedRerouteUsage}
  MaterialExpressionNamedRerouteUsage_9 {class=MaterialExpressionNamedRerouteUsage}
  MaterialExpressionNamedRerouteUsage_91 {class=MaterialExpressionNamedRerouteUsage}
  MaterialExpressionNamedRerouteUsage_92 {class=MaterialExpressionNamedRerouteUsage}
  MaterialExpressionNamedRerouteUsage_93 {class=MaterialExpressionNamedRerouteUsage}
  MaterialExpressionNamedRerouteUsage_94 {class=MaterialExpressionNamedRerouteUsage}
  MaterialExpressionNamedRerouteUsage_95 {class=MaterialExpressionNamedRerouteUsage}
  MaterialExpressionNamedRerouteUsage_96 {class=MaterialExpressionNamedRerouteUsage}
  MaterialExpressionNamedRerouteUsage_97 {class=MaterialExpressionNamedRerouteUsage}
  MaterialExpressionNamedRerouteUsage_98 {class=MaterialExpressionNamedRerouteUsage}
  MaterialExpressionNamedRerouteUsage_99 {class=MaterialExpressionNamedRerouteUsage}
  MaterialExpressionNormalize_0 {class=MaterialExpressionNormalize, inputs [VectorInput {from=MaterialExpressionWorldPosition_2, output=XYZ}]}
  MaterialExpressionNormalize_1 {class=MaterialExpressionNormalize, inputs [VectorInput {from=MaterialExpressionMultiply_107, output=""}]}
  MaterialExpressionNormalize_2 {class=MaterialExpressionNormalize, inputs [VectorInput {from=MaterialExpressionWorldPosition_3, output=XYZ}]}
  MaterialExpressionOneMinus_13 {class=MaterialExpressionOneMinus, inputs [None {from=MaterialExpressionConstant_36}]}
  MaterialExpressionOneMinus_14 {class=MaterialExpressionOneMinus, inputs [None {from=MaterialExpressionNamedRerouteUsage_119}]}
  MaterialExpressionOneMinus_15 {class=MaterialExpressionOneMinus, inputs [None {from=MaterialExpressionSaturate_56}]}
  MaterialExpressionOneMinus_20 {class=MaterialExpressionOneMinus, inputs [None {from=MaterialExpressionSaturate_57}]}
  MaterialExpressionOneMinus_21 {class=MaterialExpressionOneMinus, inputs [None {from=MaterialExpressionSaturate_58}]}
  MaterialExpressionOneMinus_22 {class=MaterialExpressionOneMinus, inputs [None {from=MaterialExpressionSaturate_59}]}
  MaterialExpressionOneMinus_23 {class=MaterialExpressionOneMinus, inputs [None {from=MaterialExpressionNamedRerouteUsage_131}]}
  MaterialExpressionOneMinus_25 {class=MaterialExpressionOneMinus, inputs [None {from=MaterialExpressionDivide_5}]}
  MaterialExpressionOneMinus_4 {class=MaterialExpressionOneMinus, inputs [None {from=MaterialExpressionAdd_2}]}
  MaterialExpressionOneMinus_6 {class=MaterialExpressionOneMinus, inputs [None {from=MaterialExpressionReroute_9}]}
  MaterialExpressionOneMinus_7 {class=MaterialExpressionOneMinus, inputs [None {from=MaterialExpressionDistance_0}]}
  MaterialExpressionOneMinus_8 {class=MaterialExpressionOneMinus, inputs [None {from=MaterialExpressionAdd_50}]}
  MaterialExpressionPixelDepth_0 {class=MaterialExpressionPixelDepth, inputs [], properties {}}
  MaterialExpressionPixelDepth_2 {class=MaterialExpressionPixelDepth, inputs [], properties {}}
  MaterialExpressionPixelDepth_3 {class=MaterialExpressionPixelDepth, inputs [], properties {}}
  MaterialExpressionPower_2 {
   class=MaterialExpressionPower
   inputs [
    Base {from=MaterialExpressionNamedRerouteUsage_91}
    Exp {from=""}
   ]
   properties {constExponent=1.3}
  }
  MaterialExpressionPower_3 {
   class=MaterialExpressionPower
   inputs [
    Base {from=MaterialExpressionOneMinus_7}
    Exp {from=""}
   ]
   properties {constExponent=4}
  }
  MaterialExpressionPower_4 {
   class=MaterialExpressionPower
   inputs [
    Base {from=MaterialExpressionDotProduct_0}
    Exp {from=MaterialExpressionScalarParameter_62}
   ]
  }
  MaterialExpressionPower_5 {
   class=MaterialExpressionPower
   inputs [
    Base {from=MaterialExpressionOneMinus_14}
    Exp {from=""}
   ]
   properties {constExponent=0.2}
  }
  MaterialExpressionPower_8 {
   class=MaterialExpressionPower
   inputs [
    Base {from=""}
    Exp {from=""}
   ]
  }
  MaterialExpressionRayTracingQualitySwitch_0 {
   class=MaterialExpressionRayTracingQualitySwitch
   inputs [
    Normal {from=MaterialExpressionMultiply_68, output=""}
    RayTraced {from=MaterialExpressionMultiply_6, output=""}
   ]
   properties {}
  }
  MaterialExpressionReroute_0 {class=MaterialExpressionReroute, inputs [None {from=MaterialExpressionSaturate_46}]}
  MaterialExpressionReroute_1 {class=MaterialExpressionReroute, inputs [None {from=MaterialExpressionReroute_2}]}
  MaterialExpressionReroute_10 {class=MaterialExpressionReroute, inputs [None {from=MaterialExpressionMultiply_104}]}
  MaterialExpressionReroute_11 {class=MaterialExpressionReroute, inputs [None {from=MaterialExpressionNamedRerouteUsage_25}]}
  MaterialExpressionReroute_2 {class=MaterialExpressionReroute, inputs [None {from=MaterialExpressionReroute_54}]}
  MaterialExpressionReroute_26 {class=MaterialExpressionReroute, inputs [None {from=MaterialExpressionAdd_12}]}
  MaterialExpressionReroute_28 {class=MaterialExpressionReroute, inputs [None {from=MaterialExpressionNamedRerouteUsage_11}]}
  MaterialExpressionReroute_29 {class=MaterialExpressionReroute, inputs [None {from=MaterialExpressionAdd_30}]}
  MaterialExpressionReroute_3 {class=MaterialExpressionReroute, inputs [None {from=MaterialExpressionMultiply_19}]}
  MaterialExpressionReroute_32 {class=MaterialExpressionReroute, inputs [None {from=MaterialExpressionConvert_6}]}
  MaterialExpressionReroute_33 {class=MaterialExpressionReroute, inputs [None {from=MaterialExpressionOneMinus_8}]}
  MaterialExpressionReroute_34 {class=MaterialExpressionReroute, inputs [None {from=MaterialExpressionNamedRerouteUsage_89}]}
  MaterialExpressionReroute_35 {class=MaterialExpressionReroute, inputs [None {from=MaterialExpressionReroute_9}]}
  MaterialExpressionReroute_36 {class=MaterialExpressionReroute, inputs [None {from=MaterialExpressionReroute_35}]}
  MaterialExpressionReroute_37 {class=MaterialExpressionReroute, inputs [None {from=MaterialExpressionConstant3Vector_0}]}
  MaterialExpressionReroute_38 {class=MaterialExpressionReroute, inputs [None {from=MaterialExpressionSaturate_41}]}
  MaterialExpressionReroute_39 {class=MaterialExpressionReroute, inputs [None {from=MaterialExpressionAdd_0}]}
  MaterialExpressionReroute_4 {class=MaterialExpressionReroute, inputs [None {from=MaterialExpressionMaterialFunctionCall_32, output=Result}]}
  MaterialExpressionReroute_40 {class=MaterialExpressionReroute, inputs [None {from=MaterialExpressionConstant3Vector_0}]}
  MaterialExpressionReroute_41 {class=MaterialExpressionReroute, inputs [None {from=MaterialExpressionConvert_5}]}
  MaterialExpressionReroute_42 {class=MaterialExpressionReroute, inputs [None {from=MaterialExpressionSaturate_47}]}
  MaterialExpressionReroute_43 {class=MaterialExpressionReroute, inputs [None {from=MaterialExpressionSaturate_47}]}
  MaterialExpressionReroute_46 {class=MaterialExpressionReroute, inputs [None {from=MaterialExpressionNamedRerouteUsage_102}]}
  MaterialExpressionReroute_47 {class=MaterialExpressionReroute, inputs [None {from=MaterialExpressionNamedRerouteUsage_102}]}
  MaterialExpressionReroute_5 {class=MaterialExpressionReroute, inputs [None {from=MaterialExpressionTextureSampleParameter2D_0, output=G}]}
  MaterialExpressionReroute_54 {class=MaterialExpressionReroute, inputs [None {from=MaterialExpressionMaterialFunctionCall_13, output=R}]}
  MaterialExpressionReroute_55 {class=MaterialExpressionReroute, inputs [None {from=MaterialExpressionReroute_1}]}
  MaterialExpressionReroute_6 {class=MaterialExpressionReroute, inputs [None {from=MaterialExpressionMaterialFunctionCall_6, output=Result}]}
  MaterialExpressionReroute_7 {class=MaterialExpressionReroute, inputs [None {from=MaterialExpressionTextureSampleParameter2D_3, output=RGBA}]}
  MaterialExpressionReroute_8 {class=MaterialExpressionReroute, inputs [None {from=MaterialExpressionLinearInterpolate_20}]}
  MaterialExpressionReroute_9 {class=MaterialExpressionReroute, inputs [None {from=MaterialExpressionTextureSample_3, output=G}]}
  MaterialExpressionRotator_0 {
   class=MaterialExpressionRotator
   inputs [
    Coordinate {from=MaterialExpressionFrac_3, output=""}
    Time {from=MaterialExpressionDivide_6, output=""}
   ]
   properties {}
  }
  MaterialExpressionRound_0 {class=MaterialExpressionRound, inputs [None {from=MaterialExpressionScalarParameter_3, output=""}], properties {}}
  MaterialExpressionRound_10 {class=MaterialExpressionRound, inputs [None {from=MaterialExpressionScalarParameter_83, output=""}], properties {}}
  MaterialExpressionRound_6 {class=MaterialExpressionRound, inputs [None {from=MaterialExpressionAbs_1, output=""}], properties {}}
  MaterialExpressionRound_7 {class=MaterialExpressionRound, inputs [None {from=MaterialExpressionAbs_5, output=""}], properties {}}
  MaterialExpressionRound_8 {class=MaterialExpressionRound, inputs [None {from=MaterialExpressionAbs_6, output=""}], properties {}}
  MaterialExpressionRound_9 {class=MaterialExpressionRound, inputs [None {from=MaterialExpressionAbs_7, output=""}], properties {}}
  MaterialExpressionSaturate_0 {class=MaterialExpressionSaturate, inputs [None {from=MaterialExpressionPower_2}]}
  MaterialExpressionSaturate_1 {class=MaterialExpressionSaturate, inputs [None {from=MaterialExpressionRound_0}]}
  MaterialExpressionSaturate_2 {class=MaterialExpressionSaturate, inputs [None {from=MaterialExpressionMaterialFunctionCall_3, output=A}]}
  MaterialExpressionSaturate_3 {class=MaterialExpressionSaturate, inputs [None {from=MaterialExpressionAdd_7}]}
  MaterialExpressionSaturate_33 {class=MaterialExpressionSaturate, inputs [None {from=MaterialExpressionLinearInterpolate_39}]}
  MaterialExpressionSaturate_36 {class=MaterialExpressionSaturate, inputs [None {from=MaterialExpressionMultiply_53}]}
  MaterialExpressionSaturate_37 {class=MaterialExpressionSaturate, inputs [None {from=MaterialExpressionDesaturation_5}]}
  MaterialExpressionSaturate_39 {class=MaterialExpressionSaturate, inputs [None {from=MaterialExpressionAdd_39}]}
  MaterialExpressionSaturate_4 {class=MaterialExpressionSaturate, inputs [None {from=MaterialExpressionDivide_0}]}
  MaterialExpressionSaturate_40 {class=MaterialExpressionSaturate, inputs [None {from=MaterialExpressionOneMinus_25}]}
  MaterialExpressionSaturate_41 {class=MaterialExpressionSaturate, inputs [None {from=MaterialExpressionMultiply_161}]}
  MaterialExpressionSaturate_42 {class=MaterialExpressionSaturate, inputs [None {from=MaterialExpressionReroute_4}]}
  MaterialExpressionSaturate_43 {class=MaterialExpressionSaturate, inputs [None {from=MaterialExpressionMultiply_144}]}
  MaterialExpressionSaturate_44 {class=MaterialExpressionSaturate, inputs [None {from=MaterialExpressionMaterialFunctionCall_31, output=Result}]}
  MaterialExpressionSaturate_46 {class=MaterialExpressionSaturate, inputs [None {from=MaterialExpressionMultiply_152}]}
  MaterialExpressionSaturate_47 {class=MaterialExpressionSaturate, inputs [None {from=MaterialExpressionConvert_5}]}
  MaterialExpressionSaturate_48 {class=MaterialExpressionSaturate, inputs [None {from=MaterialExpressionAdd_49}]}
  MaterialExpressionSaturate_49 {class=MaterialExpressionSaturate, inputs [None {from=MaterialExpressionDotProduct_0}]}
  MaterialExpressionSaturate_5 {class=MaterialExpressionSaturate, inputs [None {from=MaterialExpressionReroute_6}]}
  MaterialExpressionSaturate_50 {class=MaterialExpressionSaturate, inputs [None {from=MaterialExpressionMultiply_138}]}
  MaterialExpressionSaturate_51 {class=MaterialExpressionSaturate, inputs [None {from=MaterialExpressionNamedRerouteUsage_116}]}
  MaterialExpressionSaturate_52 {class=MaterialExpressionSaturate, inputs [None {from=MaterialExpressionScalarParameter_69}]}
  MaterialExpressionSaturate_53 {class=MaterialExpressionSaturate, inputs [None {from=MaterialExpressionMultiply_142}]}
  MaterialExpressionSaturate_54 {class=MaterialExpressionSaturate, inputs [None {from=MaterialExpressionMaterialFunctionCall_10, output=Result}]}
  MaterialExpressionSaturate_55 {class=MaterialExpressionSaturate, inputs [None {from=MaterialExpressionNamedRerouteUsage_122}]}
  MaterialExpressionSaturate_56 {class=MaterialExpressionSaturate, inputs [None {from=MaterialExpressionRound_6}]}
  MaterialExpressionSaturate_57 {class=MaterialExpressionSaturate, inputs [None {from=MaterialExpressionRound_7}]}
  MaterialExpressionSaturate_58 {class=MaterialExpressionSaturate, inputs [None {from=MaterialExpressionRound_8}]}
  MaterialExpressionSaturate_59 {class=MaterialExpressionSaturate, inputs [None {from=MaterialExpressionRound_9}]}
  MaterialExpressionSaturate_6 {class=MaterialExpressionSaturate, inputs [None {from=MaterialExpressionAdd_14}]}
  MaterialExpressionSaturate_60 {class=MaterialExpressionSaturate, inputs [None {from=MaterialExpressionScalarParameter_81}]}
  MaterialExpressionSaturate_62 {class=MaterialExpressionSaturate, inputs [None {from=MaterialExpressionLinearInterpolate_9}]}
  MaterialExpressionSaturate_63 {class=MaterialExpressionSaturate, inputs [None {from=MaterialExpressionLinearInterpolate_78}]}
  MaterialExpressionSaturate_64 {class=MaterialExpressionSaturate, inputs [None {from=MaterialExpressionNamedRerouteUsage_133}]}
  MaterialExpressionSaturate_65 {class=MaterialExpressionSaturate, inputs [None {from=MaterialExpressionRound_10}]}
  MaterialExpressionSaturate_68 {class=MaterialExpressionSaturate, inputs [None {from=MaterialExpressionScalarParameter_89}]}
  MaterialExpressionSaturate_69 {class=MaterialExpressionSaturate, inputs [None {from=MaterialExpressionScalarParameter_91}]}
  MaterialExpressionSaturate_7 {class=MaterialExpressionSaturate, inputs [None {from=MaterialExpressionAdd_17}]}
  MaterialExpressionSaturate_70 {class=MaterialExpressionSaturate, inputs [None {from=MaterialExpressionDivide_9}]}
  MaterialExpressionSaturate_71 {class=MaterialExpressionSaturate, inputs [None {from=MaterialExpressionMultiply_190}]}
  MaterialExpressionScalarParameter_0 {class=MaterialExpressionScalarParameter, properties {defaultValue=0.2, group=Brand, parameterName=Logo Glow}}
  MaterialExpressionScalarParameter_1 {class=MaterialExpressionScalarParameter, properties {defaultValue=1, group=PBR, parameterName=Emissive Intensity}}
  MaterialExpressionScalarParameter_10 {class=MaterialExpressionScalarParameter, properties {defaultValue=1, parameterName=LumenEmissiveMultiplier}}
  MaterialExpressionScalarParameter_100 {class=MaterialExpressionScalarParameter, properties {group=Brand, parameterName=Logo X Size}}
  MaterialExpressionScalarParameter_101 {class=MaterialExpressionScalarParameter, properties {defaultValue=0.5, group=Brand, parameterName=Logo Shadow Intensity}}
  MaterialExpressionScalarParameter_102 {class=MaterialExpressionScalarParameter, properties {defaultValue=10, group=Projector, parameterName=Special Mask Contrast}}
  MaterialExpressionScalarParameter_103 {class=MaterialExpressionScalarParameter, properties {defaultValue=3, group=Projector, parameterName=Special Mask Int}}
  MaterialExpressionScalarParameter_105 {class=MaterialExpressionScalarParameter, properties {defaultValue=0.8, group=HoloEffect, parameterName=Opacity DIstant, sliderMax=1}}
  MaterialExpressionScalarParameter_106 {class=MaterialExpressionScalarParameter, properties {defaultValue=0.2, group=Brand, parameterName=Logo Glow}}
  MaterialExpressionScalarParameter_11 {class=MaterialExpressionScalarParameter, properties {defaultValue=10, parameterName=UV Glitch Speed}}
  MaterialExpressionScalarParameter_12 {class=MaterialExpressionScalarParameter, properties {defaultValue=1, group=Brand, parameterName=Logo intensity}}
  MaterialExpressionScalarParameter_13 {class=MaterialExpressionScalarParameter, properties {defaultValue=400, group=Dots, parameterName=Dot Texture Distance}}
  MaterialExpressionScalarParameter_14 {
   class=MaterialExpressionScalarParameter
   properties {defaultValue=0.05, group=HolographicTexture_Glitch, parameterName=GlitchDuration, sortPriority=10}
  }
  MaterialExpressionScalarParameter_15 {
   class=MaterialExpressionScalarParameter
   properties {defaultValue=0.05, group=HolographicTexture_Glitch, parameterName=GlitchFrequency, sortPriority=10}
  }
  MaterialExpressionScalarParameter_16 {class=MaterialExpressionScalarParameter, properties {defaultValue=0.1, parameterName="Glitch Flickering (S)"}}
  MaterialExpressionScalarParameter_19 {class=MaterialExpressionScalarParameter, properties {defaultValue=10}}
  MaterialExpressionScalarParameter_2 {class=MaterialExpressionScalarParameter, properties {defaultValue=0.2, group=Color, parameterName=Base color Noise Size far}}
  MaterialExpressionScalarParameter_20 {
   class=MaterialExpressionScalarParameter
   properties {defaultValue=0.02, group=HolographicTexture_Glitch, parameterName=GlitchDuration, sortPriority=10}
  }
  MaterialExpressionScalarParameter_21 {
   class=MaterialExpressionScalarParameter
   properties {defaultValue=0.04, group=HolographicTexture_Glitch, parameterName=GlitchFrequency, sortPriority=10}
  }
  MaterialExpressionScalarParameter_22 {class=MaterialExpressionScalarParameter, properties {defaultValue=-20, group=Bumpoffset, parameterName=DFX Depth}}
  MaterialExpressionScalarParameter_23 {class=MaterialExpressionScalarParameter, properties {defaultValue=0.1, parameterName="Glitch Flickering (S)"}}
  MaterialExpressionScalarParameter_25 {class=MaterialExpressionScalarParameter, properties {defaultValue=21.8605, group=Dots, parameterName="Pixel Size (S)"}}
  MaterialExpressionScalarParameter_26 {class=MaterialExpressionScalarParameter, properties {defaultValue=1.4872, group=Dots, parameterName="Pixel Smooth (S)"}}
  MaterialExpressionScalarParameter_3 {class=MaterialExpressionScalarParameter, properties {defaultValue=1, group=Brand, parameterName=Logo intensity}}
  MaterialExpressionScalarParameter_4 {class=MaterialExpressionScalarParameter, properties {defaultValue=5, group=Color, parameterName=Base color Noise Speed}}
  MaterialExpressionScalarParameter_41 {class=MaterialExpressionScalarParameter, properties {defaultValue=0.3, group=HoloEffect, parameterName=Logo Opacity, sliderMax=1}}
  MaterialExpressionScalarParameter_47 {class=MaterialExpressionScalarParameter, properties {defaultValue=1.3, group=Brand, parameterName=Logo Background Intensity}}
  MaterialExpressionScalarParameter_5 {class=MaterialExpressionScalarParameter, properties {defaultValue=40, group=Dots, parameterName="Pixel Amount (S)"}}
  MaterialExpressionScalarParameter_50 {class=MaterialExpressionScalarParameter, properties {defaultValue=1, group=Dots, parameterName=Dots Detail Emissive}}
  MaterialExpressionScalarParameter_51 {class=MaterialExpressionScalarParameter, properties {defaultValue=1, group=Dots, parameterName=Dots Int}}
  MaterialExpressionScalarParameter_6 {class=MaterialExpressionScalarParameter, properties {defaultValue=0.8, group=HoloEffect, parameterName=Opacity Color 2, sliderMax=1}}
  MaterialExpressionScalarParameter_62 {class=MaterialExpressionScalarParameter, properties {defaultValue=1, group=PBR, parameterName=Fresnel Emissive Fade}}
  MaterialExpressionScalarParameter_65 {class=MaterialExpressionScalarParameter, properties {defaultValue=2, parameterName=Normal Strenght}}
  MaterialExpressionScalarParameter_66 {class=MaterialExpressionScalarParameter, properties {defaultValue=0.3616, group=PBR, parameterName=Spec Max, sliderMax=1}}
  MaterialExpressionScalarParameter_68 {class=MaterialExpressionScalarParameter, properties {defaultValue=0.1118, group=PBR, parameterName=Rough}}
  MaterialExpressionScalarParameter_69 {class=MaterialExpressionScalarParameter, properties {defaultValue=0.6427, group=PBR, parameterName=Spec Noise}}
  MaterialExpressionScalarParameter_7 {class=MaterialExpressionScalarParameter, properties {defaultValue=0.8, group=HoloEffect, parameterName=Opacity Color 1, sliderMax=1}}
  MaterialExpressionScalarParameter_70 {class=MaterialExpressionScalarParameter, properties {group=PBR, parameterName=Spec Min, sliderMax=1}}
  MaterialExpressionScalarParameter_72 {class=MaterialExpressionScalarParameter, properties {defaultValue=0.3593, group=PBR, parameterName=Rough Dot}}
  MaterialExpressionScalarParameter_73 {class=MaterialExpressionScalarParameter, properties {defaultValue=1.4872, group=Dots, parameterName="Pixel Smooth (S)"}}
  MaterialExpressionScalarParameter_74 {class=MaterialExpressionScalarParameter, properties {defaultValue=11, group=Dots, parameterName=Pixel Size}}
  MaterialExpressionScalarParameter_76 {class=MaterialExpressionScalarParameter, properties {defaultValue=1, group=Projector, parameterName=Projector L1, sortPriority=-4}}
  MaterialExpressionScalarParameter_77 {class=MaterialExpressionScalarParameter, properties {group=Projector, parameterName=Projector L2, sortPriority=-3}}
  MaterialExpressionScalarParameter_78 {class=MaterialExpressionScalarParameter, properties {group=Projector, parameterName=Projector L3, sortPriority=-2}}
  MaterialExpressionScalarParameter_79 {class=MaterialExpressionScalarParameter, properties {group=Projector, parameterName=Projector L4, sortPriority=-1}}
  MaterialExpressionScalarParameter_8 {class=MaterialExpressionScalarParameter, properties {defaultValue=1, group=Dots, parameterName=Dots BG Int}}
  MaterialExpressionScalarParameter_80 {class=MaterialExpressionScalarParameter, properties {defaultValue=-0.0297, group=Bumpoffset, parameterName="Max Depth (Layer 4)"}}
  MaterialExpressionScalarParameter_81 {class=MaterialExpressionScalarParameter, properties {defaultValue=0.9099, group=Bumpoffset, parameterName="Depth (Layer 3)", sliderMax=1}}
  MaterialExpressionScalarParameter_82 {class=MaterialExpressionScalarParameter, properties {defaultValue=1, group=Brand, parameterName=Logo intensity}}
  MaterialExpressionScalarParameter_83 {class=MaterialExpressionScalarParameter, properties {group=Projector, parameterName="SFX Metalic Mask (0/1)", sliderMax=1}}
  MaterialExpressionScalarParameter_86 {class=MaterialExpressionScalarParameter, properties {defaultValue=125, parameterName=Triplanar Scale}}
  MaterialExpressionScalarParameter_87 {class=MaterialExpressionScalarParameter, properties {defaultValue=20, parameterName=Triplanar Fade}}
  MaterialExpressionScalarParameter_88 {class=MaterialExpressionScalarParameter, properties {defaultValue=1, group=Brand, parameterName=Logo intensity}}
  MaterialExpressionScalarParameter_89 {class=MaterialExpressionScalarParameter, properties {defaultValue=0.3, group=Bumpoffset, parameterName="Depth Logo (Layer 2)", sliderMax=1}}
  MaterialExpressionScalarParameter_9 {class=MaterialExpressionScalarParameter, properties {defaultValue=13, parameterName=UV Glitch Size}}
  MaterialExpressionScalarParameter_91 {class=MaterialExpressionScalarParameter, properties {defaultValue=0.92, group=Bumpoffset, parameterName="Depth (Layer 3)", sliderMax=1}}
  MaterialExpressionScalarParameter_92 {class=MaterialExpressionScalarParameter, properties {defaultValue=800, group=Dots, parameterName=Dot Texture Distance}}
  MaterialExpressionScalarParameter_95 {class=MaterialExpressionScalarParameter, properties {bUseCustomPrimitiveData=true, parameterName=Dots X Size}}
  MaterialExpressionScalarParameter_96 {class=MaterialExpressionScalarParameter, properties {bUseCustomPrimitiveData=true, parameterName=Dots Y Size, primitiveDataIndex=1}}
  MaterialExpressionScalarParameter_97 {class=MaterialExpressionScalarParameter, properties {group=Brand, parameterName=Logo Y Pos, primitiveDataIndex=1}}
  MaterialExpressionScalarParameter_98 {class=MaterialExpressionScalarParameter, properties {group=Brand, parameterName=Logo Y Size}}
  MaterialExpressionScalarParameter_99 {class=MaterialExpressionScalarParameter, properties {group=Brand, parameterName=Logo X Pos}}
  MaterialExpressionSetMaterialAttributes_1 {
   class=MaterialExpressionSetMaterialAttributes
   inputs [
    MaterialAttributes {from="", output=""}
    BaseColor {from=MaterialExpressionMultiply_10, output=""}
    Metallic {from=MaterialExpressionGetMaterialAttributes_3, output=Metallic}
    Specular {from=MaterialExpressionGetMaterialAttributes_1, output=Specular}
    Roughness {from=MaterialExpressionGetMaterialAttributes_1, output=Specular}
    Normal {from=MaterialExpressionMaterialFunctionCall_0, output=Result}
    EmissiveColor {from=MaterialExpressionStaticSwitchParameter_17, output=""}
    Opacity {from=MaterialExpressionGetMaterialAttributes_3, output=Metallic}
   ]
   properties {attributeSetTypes [
     #1
     #2
     #3
     #4
     #5
     #6
     #7
    ]}
  }
  MaterialExpressionSine_2 {class=MaterialExpressionSine, inputs [None {from=MaterialExpressionDivide_1}]}
  MaterialExpressionSmoothStep_1 {
   class=MaterialExpressionSmoothStep
   inputs [
    Min {from=MaterialExpressionConstant_29, output=""}
    Max {from=MaterialExpressionConstant_30, output=""}
    Value {from=MaterialExpressionMultiply_115, output=""}
   ]
   properties {}
  }
  MaterialExpressionStaticComponentMaskParameter_0 {
   class=MaterialExpressionStaticComponentMaskParameter
   inputs [None {from=MaterialExpressionMaterialFunctionCall_47, output=Result}]
   properties {defaultG=true, group=Projector, parameterName=Mask Channel}
  }
  MaterialExpressionStaticSwitchParameter_0 {
   class=MaterialExpressionStaticSwitchParameter
   inputs [
    True {from=MaterialExpressionLinearInterpolate_7, output=""}
    False {from=MaterialExpressionLinearInterpolate_5, output=""}
   ]
   properties {group=Projector, parameterName=UseProjectorDFX, sortPriority=-100}
  }
  MaterialExpressionStaticSwitchParameter_1 {
   class=MaterialExpressionStaticSwitchParameter
   inputs [
    True {from=MaterialExpressionNamedRerouteUsage_130, output=""}
    False {from=MaterialExpressionConstant3Vector_1, output=""}
   ]
   properties {group=Projector, parameterName=Projector Emissive?, sortPriority=-99}
  }
  MaterialExpressionStaticSwitchParameter_10 {
   class=MaterialExpressionStaticSwitchParameter
   inputs [
    True {from=MaterialExpressionLinearInterpolate_23, output=""}
    False {from=MaterialExpressionNamedRerouteUsage_41, output=""}
   ]
   properties {defaultValue=true, parameterName=Mask Maker Take Over}
  }
  MaterialExpressionStaticSwitchParameter_11 {
   class=MaterialExpressionStaticSwitchParameter
   inputs [
    True {from=MaterialExpressionLinearInterpolate_14, output=""}
    False {from=MaterialExpressionNamedRerouteUsage_44, output=""}
   ]
   properties {defaultValue=true, parameterName=Mask Maker Take Over}
  }
  MaterialExpressionStaticSwitchParameter_12 {
   class=MaterialExpressionStaticSwitchParameter
   inputs [
    True {from=MaterialExpressionLinearInterpolate_13, output=""}
    False {from=MaterialExpressionNamedRerouteUsage_43, output=""}
   ]
   properties {defaultValue=true, parameterName=Mask Maker Take Over}
  }
  MaterialExpressionStaticSwitchParameter_13 {
   class=MaterialExpressionStaticSwitchParameter
   inputs [
    True {from=MaterialExpressionMultiply_147, output=""}
    False {from=MaterialExpressionMax_5, output=""}
   ]
   properties {defaultValue=true, group=Projector, parameterName=Use SFX Mask Glass, sortPriority=-99}
  }
  MaterialExpressionStaticSwitchParameter_14 {
   class=MaterialExpressionStaticSwitchParameter
   inputs [
    True {from=MaterialExpressionNamedRerouteUsage_55, output=""}
    False {from=MaterialExpressionNamedRerouteDeclaration_52, output=""}
   ]
   properties {group=Projector, parameterName=Mask Maker Take Over, sortPriority=1}
  }
  MaterialExpressionStaticSwitchParameter_15 {
   class=MaterialExpressionStaticSwitchParameter
   inputs [
    True {from=MaterialExpressionNamedRerouteUsage_56, output=""}
    False {from=MaterialExpressionConvert_2, output=""}
   ]
   properties {group=Projector, parameterName=Mask Maker Take Over, sortPriority=1}
  }
  MaterialExpressionStaticSwitchParameter_16 {
   class=MaterialExpressionStaticSwitchParameter
   inputs [
    True {from=MaterialExpressionBlendMaterialAttributes_0, output=""}
    False {from=MaterialExpressionNamedRerouteDeclaration_59, output=""}
   ]
   properties {group=Projector, parameterName=UseGlassProjector, sortPriority=-100}
  }
  MaterialExpressionStaticSwitchParameter_17 {
   class=MaterialExpressionStaticSwitchParameter
   inputs [
    True {from=MaterialExpressionGetMaterialAttributes_1, output=EmissiveColor}
    False {from=MaterialExpressionMultiply_153, output=""}
   ]
   properties {group=Projector, parameterName=Projector Emissive?, sortPriority=-99}
  }
  MaterialExpressionStaticSwitchParameter_18 {
   class=MaterialExpressionStaticSwitchParameter
   inputs [
    True {from=MaterialExpressionNamedRerouteUsage_141, output=""}
    False {from=MaterialExpressionLinearInterpolate_73, output=""}
   ]
   properties {group=Projector, parameterName=Mask Maker Take Over, sortPriority=1}
  }
  MaterialExpressionStaticSwitchParameter_19 {
   class=MaterialExpressionStaticSwitchParameter
   inputs [
    True {from=MaterialExpressionMaterialProxyReplace_0, output=""}
    False {from=MaterialExpressionConstant_38, output=""}
   ]
   properties {group=Projector, parameterName=Projector Emissive?, sortPriority=-99}
  }
  MaterialExpressionStaticSwitchParameter_2 {
   class=MaterialExpressionStaticSwitchParameter
   inputs [
    True {from=MaterialExpressionStaticSwitchParameter_33, output=""}
    False {from=MaterialExpressionMaterialFunctionCall_15, output=Result}
   ]
   properties {group=Projector, parameterName=Mask Maker Take Over, sortPriority=1}
  }
  MaterialExpressionStaticSwitchParameter_20 {
   class=MaterialExpressionStaticSwitchParameter
   inputs [
    True {from=MaterialExpressionMultiply_130, output=""}
    False {from=MaterialExpressionConstant_40, output=""}
   ]
   properties {group=Projector, parameterName=Projector Emissive?, sortPriority=-99}
  }
  MaterialExpressionStaticSwitchParameter_21 {
   class=MaterialExpressionStaticSwitchParameter
   inputs [
    True {from=MaterialExpressionStaticSwitchParameter_26, output=""}
    False {from=MaterialExpressionConstant_41, output=""}
   ]
   properties {group=Projector, parameterName=Projector Emissive?, sortPriority=-99}
  }
  MaterialExpressionStaticSwitchParameter_22 {
   class=MaterialExpressionStaticSwitchParameter
   inputs [
    True {from=MaterialExpressionStaticSwitchParameter_25, output=""}
    False {from=MaterialExpressionConstant_42, output=""}
   ]
   properties {group=Projector, parameterName=Projector Emissive?, sortPriority=-99}
  }
  MaterialExpressionStaticSwitchParameter_23 {
   class=MaterialExpressionStaticSwitchParameter
   inputs [
    True {from=MaterialExpressionMultiply_113, output=""}
    False {from=MaterialExpressionConstant_43, output=""}
   ]
   properties {group=Projector, parameterName=Projector Emissive?, sortPriority=-99}
  }
  MaterialExpressionStaticSwitchParameter_24 {
   class=MaterialExpressionStaticSwitchParameter
   inputs [
    True {from=MaterialExpressionLinearInterpolate_12, output=""}
    False {from=MaterialExpressionConstant_1, output=""}
   ]
   properties {parameterName=Mask Maker Take Over}
  }
  MaterialExpressionStaticSwitchParameter_25 {
   class=MaterialExpressionStaticSwitchParameter
   inputs [
    True {from=MaterialExpressionStaticSwitchParameter_8, output=""}
    False {from=MaterialExpressionLinearInterpolate_54, output=""}
   ]
   properties {parameterName=Mask Maker Take Over}
  }
  MaterialExpressionStaticSwitchParameter_26 {
   class=MaterialExpressionStaticSwitchParameter
   inputs [
    True {from=MaterialExpressionStaticSwitchParameter_7, output=""}
    False {from=MaterialExpressionLinearInterpolate_55, output=""}
   ]
   properties {parameterName=Mask Maker Take Over}
  }
  MaterialExpressionStaticSwitchParameter_27 {
   class=MaterialExpressionStaticSwitchParameter
   inputs [
    True {from=MaterialExpressionLinearInterpolate_65, output=""}
    False {from=MaterialExpressionMultiply_139, output=""}
   ]
   properties {group=Projector, parameterName=Projector Emissive?, sortPriority=-99}
  }
  MaterialExpressionStaticSwitchParameter_28 {
   class=MaterialExpressionStaticSwitchParameter
   inputs [
    True {from=MaterialExpressionLinearInterpolate_66, output=""}
    False {from=MaterialExpressionLinearInterpolate_59, output=""}
   ]
   properties {group=Projector, parameterName=Projector Emissive?, sortPriority=-99}
  }
  MaterialExpressionStaticSwitchParameter_29 {
   class=MaterialExpressionStaticSwitchParameter
   inputs [
    True {from=MaterialExpressionMultiply_35, output=""}
    False {from=MaterialExpressionMultiply_173, output=""}
   ]
   properties {parameterName=Mask Maker Take Over}
  }
  MaterialExpressionStaticSwitchParameter_3 {
   class=MaterialExpressionStaticSwitchParameter
   inputs [
    True {from=MaterialExpressionConstant3Vector_0, output=""}
    False {from=MaterialExpressionMaterialFunctionCall_17, output=Result}
   ]
   properties {group=Projector, parameterName=Projector Emissive?, sortPriority=-99}
  }
  MaterialExpressionStaticSwitchParameter_30 {
   class=MaterialExpressionStaticSwitchParameter
   inputs [
    True {from=MaterialExpressionSetMaterialAttributes_1, output=""}
    False {from=MaterialExpressionNamedRerouteUsage_132, output=""}
   ]
   properties {defaultValue=true, group=Projector, parameterName=Use SFX Mask Glass, sortPriority=-99}
  }
  MaterialExpressionStaticSwitchParameter_31 {
   class=MaterialExpressionStaticSwitchParameter
   inputs [
    True {from=MaterialExpressionNamedRerouteUsage_74, output=""}
    False {from=MaterialExpressionNamedRerouteUsage_70, output=""}
   ]
   properties {group=Projector, parameterName=UseProjectorDFX, sortPriority=-100}
  }
  MaterialExpressionStaticSwitchParameter_32 {
   class=MaterialExpressionStaticSwitchParameter
   inputs [
    True {from=MaterialExpressionLinearInterpolate_26, output=""}
    False {from=MaterialExpressionVectorParameter_0, output=RGB}
   ]
   properties {group=Projector, parameterName=Mask Maker Take Over, sortPriority=1}
  }
  MaterialExpressionStaticSwitchParameter_33 {
   class=MaterialExpressionStaticSwitchParameter
   inputs [
    True {from=MaterialExpressionNamedRerouteUsage_3, output=""}
    False {from=MaterialExpressionTextureSample_6, output=RGBA}
   ]
   properties {group=Projector, parameterName=UseProjectorDFX, sortPriority=-100}
  }
  MaterialExpressionStaticSwitchParameter_34 {
   class=MaterialExpressionStaticSwitchParameter
   inputs [
    True {from=MaterialExpressionNamedRerouteUsage_4, output=""}
    False {from=MaterialExpressionTextureSample_4, output=RGB}
   ]
   properties {group=Projector, parameterName=UseProjectorDFX, sortPriority=-100}
  }
  MaterialExpressionStaticSwitchParameter_36 {
   class=MaterialExpressionStaticSwitchParameter
   inputs [
    True {from=MaterialExpressionMultiply_27, output=""}
    False {from=MaterialExpressionMultiply_174, output=""}
   ]
   properties {group=Projector, parameterName=Mask Maker Take Over, sortPriority=1}
  }
  MaterialExpressionStaticSwitchParameter_4 {
   class=MaterialExpressionStaticSwitchParameter
   inputs [
    True {from=MaterialExpressionTextureSampleParameter2D_2, output=RGBA}
    False {from=MaterialExpressionReroute_7, output=""}
   ]
   properties {group=Projector, parameterName=Mask Maker Take Over, sortPriority=1}
  }
  MaterialExpressionStaticSwitchParameter_5 {
   class=MaterialExpressionStaticSwitchParameter
   inputs [
    True {from="", output=""}
    False {from="", output=""}
   ]
   properties {group=Projector, parameterName=Mask Maker Take Over, sortPriority=1}
  }
  MaterialExpressionStaticSwitchParameter_6 {
   class=MaterialExpressionStaticSwitchParameter
   inputs [
    True {from=MaterialExpressionNamedRerouteUsage_39, output=""}
    False {from=MaterialExpressionComponentMask_2, output=""}
   ]
   properties {defaultValue=true, group=Projector, parameterName=Mask Maker Take Over, sortPriority=1}
  }
  MaterialExpressionStaticSwitchParameter_7 {
   class=MaterialExpressionStaticSwitchParameter
   inputs [
    True {from=MaterialExpressionMultiply_5, output=""}
    False {from=MaterialExpressionNamedRerouteUsage_12, output=""}
   ]
   properties {group=Projector, parameterName=Mask Maker Take Over, sortPriority=1}
  }
  MaterialExpressionStaticSwitchParameter_8 {
   class=MaterialExpressionStaticSwitchParameter
   inputs [
    True {from=MaterialExpressionMultiply_15, output=""}
    False {from=MaterialExpressionNamedRerouteUsage_16, output=""}
   ]
   properties {group=Projector, parameterName=Mask Maker Take Over, sortPriority=1}
  }
  MaterialExpressionStaticSwitchParameter_9 {
   class=MaterialExpressionStaticSwitchParameter
   inputs [
    True {from=MaterialExpressionLinearInterpolate_24, output=""}
    False {from=MaterialExpressionNamedRerouteUsage_40, output=""}
   ]
   properties {defaultValue=true, parameterName=Mask Maker Take Over}
  }
  MaterialExpressionSubtract_0 {
   class=MaterialExpressionSubtract
   inputs [
    A {from=MaterialExpressionNamedRerouteUsage_102}
    B {from=MaterialExpressionMultiply_118}
   ]
  }
  MaterialExpressionSubtract_1 {
   class=MaterialExpressionSubtract
   inputs [
    A {from=MaterialExpressionLinearInterpolate_80}
    B {from=MaterialExpressionMultiply_207}
   ]
  }
  MaterialExpressionSubtract_2 {
   class=MaterialExpressionSubtract
   inputs [
    A {from=MaterialExpressionNamedRerouteUsage_91}
    B {from=MaterialExpressionNamedRerouteUsage_2}
   ]
  }
  MaterialExpressionSubtract_3 {
   class=MaterialExpressionSubtract
   inputs [
    A {from=MaterialExpressionNamedRerouteUsage_179}
    B {from=MaterialExpressionNamedRerouteUsage_180}
   ]
  }
  MaterialExpressionTextureCoordinate_2 {class=MaterialExpressionTextureCoordinate}
  MaterialExpressionTextureCoordinate_3 {class=MaterialExpressionTextureCoordinate}
  MaterialExpressionTextureCoordinate_5 {class=MaterialExpressionTextureCoordinate, properties {coordinateIndex=1}}
  MaterialExpressionTextureCoordinate_7 {class=MaterialExpressionTextureCoordinate}
  MaterialExpressionTextureCoordinate_9 {class=MaterialExpressionTextureCoordinate}
  MaterialExpressionTextureObjectParameter_1 {
   class=MaterialExpressionTextureObjectParameter
   inputs [
    UVs {from="", output=""}
    Apply View MipBias {from="", output=""}
   ]
   properties {group=Brand, parameterName=Logo, texture=/Hermes/Techart/UtopiaEmissiveSurfaces/Glass_4Layer/Base_Instances/ProjectorDFX/T_BrandFoeBurguer}
  }
  MaterialExpressionTextureSampleParameter2D_0 {
   class=MaterialExpressionTextureSampleParameter2D
   inputs [
    UVs {from=MaterialExpressionMaterialFunctionCall_2, output=UVCoord}
    Apply View MipBias {from="", output=""}
   ]
   properties {parameterName=DFX_Texture, texture=/Game/Global/Base4Layers_2022/Textures/T-GraffitiSample_D}
  }
  MaterialExpressionTextureSampleParameter2D_2 {
   class=MaterialExpressionTextureSampleParameter2D
   inputs [
    UVs {from=MaterialExpressionStaticSwitchParameter_18, output=""}
    Apply View MipBias {from="", output=""}
   ]
   properties {parameterName=DFX_Texture, texture=&T-GraffitiSample_D}
  }
  MaterialExpressionTextureSampleParameter2D_3 {
   class=MaterialExpressionTextureSampleParameter2D
   inputs [
    UVs {from=MaterialExpressionNamedRerouteUsage_142, output=""}
    Apply View MipBias {from="", output=""}
   ]
   properties {parameterName=DFX_Texture, texture=&T-GraffitiSample_D}
  }
  MaterialExpressionTextureSample_0 {
   class=MaterialExpressionTextureSample
   inputs [
    UVs {from=MaterialExpressionAppendVector_2}
    Tex {from=""}
    Apply View MipBias {from=""}
   ]
   properties {samplerType=SAMPLERTYPE_LinearColor, texture=/Game/Characters/Player/Male/Medium/Bodies/M_MED_BiruFang/Textures/T_BiruFang_Noise}
  }
  MaterialExpressionTextureSample_3 {
   class=MaterialExpressionTextureSample
   inputs [
    UVs {from=MaterialExpressionAppendVector_2}
    Tex {from=""}
    "DDX(UVs)" {from=MaterialExpressionNamedRerouteUsage_137}
    "DDY(UVs)" {from=MaterialExpressionNamedRerouteUsage_138}
    Apply View MipBias {from=""}
   ]
   properties {
    mipValueMode=TMVM_Derivative
    samplerType=SAMPLERTYPE_LinearColor
    texture=&T_BiruFang_Noise
   }
  }
  MaterialExpressionTextureSample_4 {
   class=MaterialExpressionTextureSample
   inputs [
    UVs {from=MaterialExpressionStaticSwitchParameter_15}
    Tex {from=MaterialExpressionTextureObjectParameter_1}
    Apply View MipBias {from=""}
   ]
   properties {constMipValue=3, samplerSource=SSM_Clamp_WorldGroupSettings, texture=/Hermes/Sets/UtopiaCity/Textures/T_UtopiaCity_Puppuccino_Screen_M}
  }
  MaterialExpressionTextureSample_6 {
   class=MaterialExpressionTextureSample
   inputs [
    UVs {from=MaterialExpressionStaticSwitchParameter_14}
    Tex {from=MaterialExpressionTextureObjectParameter_1}
    Apply View MipBias {from=""}
   ]
   properties {samplerSource=SSM_Clamp_WorldGroupSettings, texture=&T_UtopiaCity_Puppuccino_Screen_M}
  }
  MaterialExpressionTextureSample_7 {
   class=MaterialExpressionTextureSample
   inputs [
    UVs {from=MaterialExpressionMultiply_141}
    Tex {from=""}
    Apply View MipBias {from=""}
   ]
   properties {texture=&/T_ProjectorEmissiveDot}
  }
  MaterialExpressionTextureSample_8 {
   class=MaterialExpressionTextureSample
   inputs [
    UVs {from=MaterialExpressionRotator_0}
    Tex {from=""}
    Apply View MipBias {from=""}
   ]
   properties {texture=&T_ProjectorEmissiveDot}
  }
  MaterialExpressionTime_0 {class=MaterialExpressionTime}
  MaterialExpressionTime_3 {class=MaterialExpressionTime}
  MaterialExpressionTime_4 {class=MaterialExpressionTime}
  MaterialExpressionVectorParameter_0 {class=MaterialExpressionVectorParameter, properties {group=Brand, parameterName=Logo Color, defaultValue=<rgba 0.0541 0.4856 1 1>}}
  MaterialExpressionVectorParameter_1 {
   class=MaterialExpressionVectorParameter
   properties {group=A__EmissiveFilters, parameterName=Emissive_Color_Multiplier, defaultValue=<rgba 1 1 1 1>}
  }
  MaterialExpressionVectorParameter_12 {
   class=MaterialExpressionVectorParameter
   properties {group=Brand, parameterName=Logo Color MASK MAKER, defaultValue=<rgba 0.1719 0 0.0136 1>}
  }
  MaterialExpressionVectorParameter_13 {class=MaterialExpressionVectorParameter, properties {group=Brand, parameterName=Logo Color 2, defaultValue=<rgba 1 1 1 1>}}
  MaterialExpressionVectorParameter_14 {class=MaterialExpressionVectorParameter, properties {group=Color, parameterName=Glass Color Noise 1, defaultValue=<rgba 0 0.1812 0.2747 1>}}
  MaterialExpressionVectorParameter_16 {class=MaterialExpressionVectorParameter, properties {group=Color, parameterName=Glass Color Noise 2, defaultValue=<rgba 0 0.2623 0.3763 1>}}
  MaterialExpressionVectorParameter_21 {
   class=MaterialExpressionVectorParameter
   properties {group=Color, parameterName=Glass Color MASK MAKER 2, defaultValue=<rgba 0.026 0 0.0067 1>}
  }
  MaterialExpressionVectorParameter_3 {class=MaterialExpressionVectorParameter, properties {group=Color, parameterName=Glass Color Noise 2, defaultValue=<rgba 0 0.2623 0.3763 1>}}
  MaterialExpressionVectorParameter_5 {class=MaterialExpressionVectorParameter, properties {group=Brand, parameterName=Logo Color, defaultValue=<rgba 0.0541 0.4856 1 1>}}
  MaterialExpressionVertexInterpolator_1 {class=MaterialExpressionVertexInterpolator, inputs [VS {from=MaterialExpressionMaterialFunctionCall_36, output=UVCoord}], properties {}}
  MaterialExpressionVertexInterpolator_2 {class=MaterialExpressionVertexInterpolator, inputs [VS {from=MaterialExpressionMaterialFunctionCall_4, output=UVCoord}], properties {}}
  MaterialExpressionVertexInterpolator_3 {class=MaterialExpressionVertexInterpolator, inputs [VS {from=MaterialExpressionSaturate_49, output=""}], properties {}}
  MaterialExpressionVertexNormalWS_1 {class=MaterialExpressionVertexNormalWS, inputs [], properties {}}
  MaterialExpressionVertexNormalWS_3 {class=MaterialExpressionVertexNormalWS, inputs [], properties {}}
  MaterialExpressionWorldPosition_1 {class=MaterialExpressionWorldPosition, inputs [], properties {}}
  MaterialExpressionWorldPosition_2 {class=MaterialExpressionWorldPosition, inputs [], properties {worldPositionShaderOffset=WPT_CameraRelative}}
  MaterialExpressionWorldPosition_3 {class=MaterialExpressionWorldPosition, inputs [], properties {worldPositionShaderOffset=WPT_CameraRelative}}
 }}
